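# Niagara effect for the EmeraldGlassTransform "Might" emote, merging several transformation layers into one system.
# Emitters layer gold aura meshes, mesh bursts, GPU sparkles, flare and glow sprites, and a distance-culled, delayed "HA" burst.
# Combined emitters select per-particle values by mesh or partition index, or write extra color and size attributes alongside multiple renderers.
# Time-of-day and front-end-versus-gameplay modifiers adjust brightness, alpha and material parameters.
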NiagaraSystem /BRCosmetics/Animation/Game/MainPlayer/Emotes/EmeraldGlassTransform/FX/NS_Emote_EmeraldGlassTransform_Might_COMBINED
dependencies {
 usedDataInterfaces [
  &NiagaraDataInterfaceCurve
  &NiagaraDataInterfaceColorCurve
  &NiagaraDataInterfaceVector2DCurve
  &NiagaraDataInterfaceArrayFloat3
  &NiagaraDataInterfaceArrayFloat
  &NiagaraDataInterfaceVectorCurve
  &NiagaraDataInterfaceArrayColor
 ]
 usedDynamicInputs []
 usedModules [
  &SystemState
  &self:SystemSpawnScript.SystemScriptSource.SystemScriptGraph.NiagaraNodeAssignment_0.SetVariables_#1
 ]
}
emitters {
 AURA_Combined {
  emitterUpdateScript [
   &EmitterState()
   &SpawnRate(
    SpawnRate=&FloatFromCurve(
     FloatCurve=DataInterface() {propertyValues {Curve {keys [<at 0 1>,<at 0.7 1>,<at 0.8 0>]}}}
     CurveIndex=$Emitter.NormalizedLoopAge
     Scale Curve=8
    )
    Use Spawn Probability=false
   )
  ]
  particleSpawnScript [
   /Niagara/Modules/Spawn/Initialization/InitializeParticle(
    Color=<rgba 1 0.8445 0.1146 4>
    Lifetime=0.2
    Mesh Scale=&RandomRangeVector(Minimum=<2.2 2.7 2.7>, Maximum=<2.4 2.5 2.5>)
    Module.Write Color=true
    Module.Write Lifetime=true
    Module.Write Mass=true
    Module.Write Material Random=true
    Module.Write Position=true
    Module.Write Scale=true
    Module.Write SpriteRotation=false
    Module.Write SpriteSize=true
    Position=&Add_Vector(A=&SimulationPosition(), B=<0 0 -2>)
    Sprite Size=<5 5>
   )
   &FN_Mesh_Attributes(
    Color=<rgba 2 1.44 0.6 1>
    Create Color Attribute=true
    "Create Normalized Age "=true
    Create Orientation Attribute=true
    Create Scale Attribute=true
    Lifetime:NiagaraFloat=&RandomRangeFloat(Minimum=0.3, Maximum=0.4)
    Scale=&Multiply_VectorByFloat(Vector=&RandomRangeVector(Minimum=<1 1 1>, Maximum=<0.5 1 1>), Float=2)
    Write to Module Namespace=true
   )
   /SetVariables(Particles.Pos_AuraRing:NiagaraPosition=&Add_Vector(A=&SimulationPosition(), B=<0 0 10>)) {isSetParameters=true}
   &InitialMeshOrientation(
    Rotation=&RandomRangeVector(Minimum=<0 0.25 0>, Maximum=<0 0.25 0>)
    Use Orientation Vector=true
    Use Rotation Vector=true
   )
   &InitialMeshOrientation(
    Rotation=&RandomRangeVector(Minimum=<0 0 0>, Maximum=<0 0 0.5>)
    Use Orientation Vector=true
    Use Rotation Vector=true
   ) {module=InitialMeshOrientation001}
  ]
  particleUpdateScript [
   &ParticleState()
   /SetVariables(Particles.Scale_AuraRing:Vector3f=&Multiply_Vector(
     A=$Particles.FN_Mesh_Attributes.Scale
     B=&VectorFromFloat(Value=&FloatFromCurve(
       FloatCurve=DataInterface() {propertyValues {Curve {keys [{interpMode=RCIM_Cubic, tangentMode=RCTM_User, time=0, value=0.5}, {interpMode=RCIM_Cubic, tangentMode=RCTM_User, time=1, value=2}]}}}
       Scale Curve=0.8
      ))
    )) {isSetParameters=true}
   /SetVariables(Particles.ScaleColor_AuraRing:LinearColor=&MultiplyLinearColor(
     Linear Color=$Particles.FN_Mesh_Attributes.Color
     Scale Factor=&MakeFloat4FromVectorAndFloat(
      "Vector (RGB)"=&Multiply_VectorByFloat(
       Vector=<1 1 1>
       Float=&Multiply_Float(
        A=&FNTimeOfDayScalar(Night Scale=0.7, Morning Scale=0.3, Evening Scale=0.4)
        B=&FN_FrontEndFloatModifier(Gameplay=1, Front End=2)
       )
      )
      "Float (Alpha)"=&FloatFromCurve(
       FloatCurve=DataInterface() {propertyValues {Curve {keys [
           {interpMode=RCIM_Cubic, tangentMode=RCTM_User, time=0, value=0}
           {interpMode=RCIM_Cubic, tangentMode=RCTM_User, time=0.3, value=0.75}
           {tangentMode=RCTM_User, time=1, value=0}
          ]}}}
       Scale Curve=&FN_FrontEndFloatModifier(Gameplay=1.5, Front End=0.4)
      )
     )
    )) {isSetParameters=true}
   &ScaleMeshSize(Scale Factor=&VectorFromCurve(
     VectorCurve=DataInterface() {propertyValues {XCurve {keys [<at 0 0.5>,<at 1 1.5>]}, YCurve {keys [<at 0 0.7>,<at 1 1.6>]}, ZCurve {keys [<at 0 0.7>,<at 1 2>]}}}
     Scale Curve=&Multiply_Vector(A=<0.7 0.7 0.7>, B=$Engine.Owner.Scale)
    ))
   &SolveForcesAndVelocity(Clamp Velocity=false, Limit Acceleration=false)
   &ScaleColor(
    Scale Alpha=&Multiply_Float(
     A=1
     B=&Multiply_Float(
      A=&FloatFromCurve(FloatCurve=DataInterface() {propertyValues {Curve {keys [<at 0 0 RCIM_Cubic>,<at 0.026 1 RCIM_Cubic>,<at 1 0>]}}}, Scale Curve=2)
      B=&FNTimeOfDayScalar(Night Scale=0.5, Morning Scale=0.2, Evening Scale=0.2)
     )
    )
    ScaleA=true
    ScaleRGB=true
   )
   &ScaleMeshSize(
    Initial Mesh Scale:Vector3f=$Particles.FN_Mesh_Attributes.Scale
    Scale Factor=&VectorFromFloat(Value=&FloatFromCurve(
      FloatCurve=DataInterface() {propertyValues {Curve {keys [{interpMode=RCIM_Cubic, tangentMode=RCTM_User, time=0, value=1}, {interpMode=RCIM_Cubic, tangentMode=RCTM_User, time=1, value=1.5}]}}}
      CurveIndex=$Particles.FN_Mesh_Attributes.NormalizedAge
      Scale Curve=0.8
     ))
   ) {module=ScaleMeshSize001}
   &DynamicMaterialParameters(
    Index 0 Param 1=&FloatFromCurve(FloatCurve=DataInterface() {propertyValues {Curve {keys [<at 0 0 RCIM_Cubic>,<at 1 0.28 RCIM_Cubic>]}}})
    Index 0 Param 2=&RandomRangeFloat(Minimum=0, Maximum=1)
    Index 0 Param 3=50
    Index 0 Param 4=&FloatFromCurve(FloatCurve=DataInterface() {propertyValues {Curve {keys [<at 0 0>, {arriveTangent=-3.0374, leaveTangent=2.0371, time=0.2, value=0}, {interpMode=RCIM_Cubic, tangentMode=RCTM_Break, time=1, value=0}]}}})
    Index 1 Param 1=&FN_FrontEndFloatModifier(Gameplay=0, Front End=0.5)
    Module.Param0WriteEnabled=true
    Module.Param0WriteEnabled1=true
    Module.Param1WriteEnabled=true
    Module.Param1WriteEnabled1=true
    Module.Param2WriteEnabled=true
    Module.Param2WriteEnabled1=true
    Module.Param3WriteEnabled=true
    Module.Param3WriteEnabled1=true
    Write Parameter Index 1=true
   )
   /SetVariables(Particles.DynParam:Vector4f=&MakeVector4(
     X=&FloatFromCurve(FloatCurve=DataInterface() {propertyValues {Curve {keys [<at 0 0 RCIM_Cubic>,<at 1 0.28 RCIM_Cubic>]}}})
     Y=0.5
     Z=&FN_FrontEndFloatModifier(Gameplay=0.5, Front End=2)
     W=&FN_FrontEndFloatModifier(Gameplay=25, Front End=50)
    )) {isSetParameters=true}
  ]
  rendererClasses [&NiagaraMeshRendererProperties]
  renderers [
   {rendererClass=&NiagaraMeshRendererProperties, rendererIndex=0}
   {rendererClass=&NiagaraMeshRendererProperties, rendererIndex=1}
  ]
 }
 BoxSparkle {
  simTarget=GPUComputeSim
  emitterUpdateScript [
   &EmitterState()
   &SpawnRate(
    SpawnRate=&FloatFromCurve(
     FloatCurve=DataInterface() {propertyValues {Curve {keys [<at 0 1>,<at 0.6 1>,<at 0.7 0>]}}}
     CurveIndex=$Emitter.NormalizedLoopAge
     Scale Curve=50
    )
    Use Spawn Probability=false
   )
  ]
  particleSpawnScript [
   /Niagara/Modules/Spawn/Initialization/V2/InitializeParticle(
    Color Maximum=<rgba 8 4.4582 1.6 1>
    Color Minimum=<rgba 0.7448 0.4084 0.149 1>
    Color Mode=::NewEnumerator2
    Lifetime Max=0.6
    Lifetime Min=0.2
    Lifetime Mode=::NewEnumerator1
    Mass Max=2
    Mass Mode=::NewEnumerator3
    Position Offset=<0 0 60>
    Sprite Rotation Angle=45
    Sprite Rotation Mode=::NewEnumerator0
    Sprite Size Mode=::NewEnumerator1
    Uniform Sprite Size Max=50
    Uniform Sprite Size Min=30
    UsePositionOffset=true
    Write Material Random=true
   )
   &BoxLocation(Box Size=<200 200 200>, MaskSpawnGroup=false, Module.Override Seed=false, Rotate Box=false, Surface Only=false) {enabled=false}
   &AddVelocity(Velocity=&RandomRangeVector(Minimum=<-1 -1 -1>, Maximum=<1 1 1>), Velocity Speed Scale=500)
   &AddVelocityFromPoint(Module.Use Falloff Distance=false, Velocity Strength=&RandomRangeFloat(Minimum=100, Maximum=250)) {enabled=false}
  ]
  particleUpdateScript [
   &ParticleState()
   &Drag(Drag=&RandomRangeFloat(Minimum=2.5, Maximum=5), Use Linear Drag=true, Use Rotational Drag=false)
   &ScaleColor(
    Scale Alpha=&FloatFromCurve(FloatCurve=DataInterface() {propertyValues {Curve {keys [<at 0 0 RCIM_Cubic>, {arriveTangent=0.002, interpMode=RCIM_Cubic, leaveTangent=0.002, time=0.4, value=0.35}, <at 1.0011 0.002 RCIM_Cubic>]}}})
    ScaleA=true
    ScaleRGB=true
   )
   &SolveForcesAndVelocity(Clamp Velocity=false, Limit Acceleration=false)
   &CameraOffset(Camera Offset Amount=20)
   &ScaleSpriteSize(Uniform Curve Sprite Scale=DataInterface() {propertyValues {Curve {keys [<at 0 0 RCIM_Cubic>,<at 1 1 RCIM_Cubic>]}}})
  ]
  rendererClasses [&NiagaraSpriteRendererProperties]
  renderers [{rendererClass=&NiagaraSpriteRendererProperties}]
 }
 Burst_Combined {
  emitterUpdateScript [
   &EmitterState(Life Cycle Mode=::NewEnumerator1, Loop Behavior=::NewEnumerator2, Loop Count=6, Loop Duration=0.05, UseLoopDelay=false)
   &SpawnBurst_Instantaneous(Use Spawn Probability=false, UseLoopCountLimit=false)
  ]
  particleSpawnScript [
   /SetVariables(Particles.MeshIndex=&SelectIntFromArray(
     Direct Array Index=$Emitter.LoopCount
     Direct Array Mode=::NewEnumerator1
     Int Selection Array=DataInterface() {propertyValues {IntData [1,0,0,0,0,0]}}
    )) {isSetParameters=true}
   &InitializeParticle(
    Color=&SelectLinearColorFromArray(
     Array Sampling Mode=::NewEnumerator1
     Direct Array Index=$Particles.MeshIndex
     Color Selection Array=DataInterface() {propertyValues {ColorData [<rgba 2 6.5 20 0.1>, <rgba 10 5.2741 0 1.4>]}}
     Random Seed=?
    )
    Lifetime=&Multiply_Float(
     A=&RandomRangeFloat(Minimum=0.9, Maximum=1.1)
     B=&SelectFloatFromArray(Direct Array Index=$Particles.MeshIndex, Float Selection Array=DataInterface() {propertyValues {FloatData [0.55,0.3]}})
    )
    Mesh Scale=&Multiply_Vector(
     A=&SelectVectorFromArray(Direct Array Index=$Particles.MeshIndex, Vector Selection Array=DataInterface() {propertyValues {FloatData [<1 1 1.5>, <0.1 0.1 0.1>]}})
     B=&MakeVector(
      X=1
      Y=1
      Z=&RandomRangeFloat(
       Minimum=&SelectFloatFromArray(Direct Array Index=$Particles.MeshIndex, Float Selection Array=DataInterface() {propertyValues {FloatData [0.5,1]}})
       Maximum=1
      )
     )
    )
    Mesh Scale Mode=::NewEnumerator4
    UsePositionOffset=false
    Write Material Random=true
   )
   &BoxLocation(
    Box Offset=&SelectVectorFromArray(Direct Array Index=$Particles.MeshIndex, Vector Selection Array=DataInterface() {propertyValues {FloatData [<0 0 120>, <0 0 90>]}})
    Box Size=&SelectVectorFromArray(Direct Array Index=$Particles.MeshIndex, Vector Selection Array=DataInterface() {propertyValues {FloatData [<0 0 100>, <0 0 0>]}})
    MaskSpawnGroup=false
    Module.Override Seed=false
    Rotate Box=false
    Surface Only=false
   )
   &InitialMeshOrientation(
    Mesh Orientation Mode=::NewEnumerator2
    Orientation Coordinate Space=::Local
    Rotation=&RandomRangeVector(
     Minimum=&SelectVectorFromArray(Direct Array Index=$Particles.MeshIndex, Vector Selection Array=DataInterface() {propertyValues {FloatData [<-1 -1 -1>, <0 0.25 0>]}})
     Maximum=&SelectVectorFromArray(Direct Array Index=$Particles.MeshIndex, Vector Selection Array=DataInterface() {propertyValues {FloatData [<1 1 1>, <0 0.25 0>]}})
    )
    Use Rotation Vector=true
   )
  ]
  particleUpdateScript [
   &ParticleState()
   &DynamicMaterialParameters(
    Index 0 Param 1=&FloatFromCurve(
     FloatCurve=DataInterface() {propertyValues {Curve {keys [{interpMode=RCIM_Cubic, tangentMode=RCTM_User, time=0, value=1}, {interpMode=RCIM_Cubic, tangentMode=RCTM_User, time=1, value=0}]}}}
     Scale Curve=&SelectFloatFromArray(Direct Array Index=$Particles.MeshIndex, Float Selection Array=DataInterface() {propertyValues {FloatData [1,0]}})
    )
    Index 0 Param 2=&SelectFloatFromArray(Direct Array Index=$Particles.MeshIndex, Float Selection Array=DataInterface() {propertyValues {FloatData [0,2]}})
    Index 0 Param 3=&SelectFloatFromArray(Direct Array Index=$Particles.MeshIndex, Float Selection Array=DataInterface() {propertyValues {FloatData [0,1]}})
    Module.Param0WriteEnabled=true
    Module.Param1WriteEnabled=true
    Module.Param2WriteEnabled=true
    Module.Param3WriteEnabled=true
   )
   &ScaleMeshSize(Scale Factor=&Multiply_VectorByFloat(
     Vector=&Multiply_VectorByFloat(Vector=$Engine.Owner.Scale, Float=1)
     Float=&FloatFromCurve(
      FloatCurve=DataInterface() {propertyValues {Curve {keys [
          {arriveTangent=1.6041, interpMode=RCIM_Cubic, leaveTangent=1.6041, tangentMode=RCTM_User, time=0, value=0.4}
          {arriveTangent=0.9995, interpMode=RCIM_Cubic, leaveTangent=0.9995, tangentMode=RCTM_User, time=0.2, value=0.7}
          {arriveTangent=-0.0405, interpMode=RCIM_Cubic, leaveTangent=-0.0405, tangentMode=RCTM_User, time=1, value=1}
         ]}}}
      Scale Curve=&SelectFloatFromArray(Direct Array Index=$Particles.MeshIndex, Float Selection Array=DataInterface() {propertyValues {FloatData [2,3.25]}})
     )
    ))
   &UpdateMeshOrientation(
    Rotation Rate=1.5
    Rotation Vector=&SelectVectorFromArray(Direct Array Index=$Particles.MeshIndex, Vector Selection Array=DataInterface() {propertyValues {FloatData [<0 0 1>, <1 0 0>]}})
   )
   &ScaleColor(
    Scale Alpha=&FloatFromCurve(
     FloatCurve=DataInterface() {propertyValues {Curve {keys [
         {arriveTangent=4.1785, interpMode=RCIM_Cubic, leaveTangent=4.1785, tangentMode=RCTM_User, time=0, value=0}
         {arriveTangent=0.9995, interpMode=RCIM_Cubic, leaveTangent=0.9995, tangentMode=RCTM_User, time=0.3, value=0.625}
         {arriveTangent=-0.0405, interpMode=RCIM_Cubic, leaveTangent=-0.0405, tangentMode=RCTM_User, time=1, value=1}
        ]}}}
     Scale Curve=&FN_FrontEndFloatModifier(
      Gameplay=&SelectFloatFromArray(Direct Array Index=$Particles.MeshIndex, Float Selection Array=DataInterface() {propertyValues {FloatData [2,0.9]}})
      Front End=&SelectFloatFromArray(Direct Array Index=$Particles.MeshIndex, Float Selection Array=DataInterface() {propertyValues {FloatData [2,0.75]}})
     )
    )
    Scale RGB=&Multiply_VectorByFloat(Vector=<1 1 1>, Float=&FNTimeOfDayScalar(Night Scale=0.5, Morning Scale=0.3, Evening Scale=0.3))
    ScaleA=true
    ScaleRGB=true
   )
  ]
  rendererClasses [&NiagaraMeshRendererProperties]
  renderers [{rendererClass=&NiagaraMeshRendererProperties, rendererIndex=0}]
 }
 Decals {
  emitterUpdateScript [
   &EmitterState()
   &SpawnBurst_Instantaneous(Spawn Count=1, Use Spawn Probability=false) {module=SpawnBurst_Instantaneous001}
  ]
  particleSpawnScript [
   &PartitionParticles()
   &InitializeParticle(
    Color=&FNTimeOfDayColor(Night Color=<rgba 5 4.1702 1.5 1>, Morning Color=<rgba 10 7.629 0 1>, Day Color=<rgba 10 7.629 0 1>, Evening Color=<rgba 5 4.1702 1.5 1>)
    Lifetime=0.4
    Position Offset=&SelectVectorFromArray(
     Direct Array Index=$Particles.PartitionParticles.Partition
     Vector Selection Array=DataInterface() {propertyValues {FloatData [<0 0 110>, <0 0 10>]}}
    )
    Position Offset Coordinate Space=::Local
    Sprite Size Mode=::NewEnumerator3
    Uniform Sprite Size=&SelectFloatFromArray(Direct Array Index=$Particles.PartitionParticles.Partition, Float Selection Array=DataInterface() {propertyValues {FloatData [190,80]}})
    UsePositionOffset=true
    Write Material Random=true
   ) {module=InitializeParticle001}
  ]
  particleUpdateScript [
   &ParticleState()
   &CameraOffset(Camera Offset Amount=&SelectFloatFromArray(Direct Array Index=$Particles.PartitionParticles.Partition, Float Selection Array=DataInterface() {propertyValues {FloatData [50,30]}}))
   &DynamicMaterialParameters(
    Index 0 Param 1=2
    Index 0 Param 2=1
    Module.Param0WriteEnabled=true
    Module.Param1WriteEnabled=true
    Module.Param2WriteEnabled=true
    Module.Param3WriteEnabled=true
   )
   &ScaleColor(
    Scale Alpha=&FloatFromCurve(
     FloatCurve=DataInterface() {propertyValues {Curve {keys [<at 0 1>, {arriveTangent=-3.5366, interpMode=RCIM_Cubic, leaveTangent=-3.5366, tangentMode=RCTM_User, time=0.2, value=1}, <at 1 0>]}}}
     Scale Curve=4
    )
    Scale RGB=&MakeVectorFromLinearColorRGB(Color=&ColorFromCurve(ColorCurve=DataInterface() {propertyValues {
        AlphaCurve {keys [<at 0 1 RCIM_Cubic>]}
        BlueCurve {keys [<at 0 1 RCIM_Cubic>, {arriveTangent=4.4002, interpMode=RCIM_Cubic, leaveTangent=4.4002, tangentMode=RCTM_User, time=1, value=3}]}
        GreenCurve {keys [<at 0 1 RCIM_Cubic>, {arriveTangent=4.4002, interpMode=RCIM_Cubic, leaveTangent=4.4002, tangentMode=RCTM_User, time=1, value=3}]}
        RedCurve {keys [<at 0 1 RCIM_Cubic>, {arriveTangent=4.4002, interpMode=RCIM_Cubic, leaveTangent=4.4002, tangentMode=RCTM_User, time=1, value=3}]}
       }}))
    ScaleA=true
    ScaleRGB=true
   )
  ]
  rendererClasses [&NiagaraSpriteRendererProperties]
  renderers [{rendererClass=&NiagaraSpriteRendererProperties}]
 }
 Explode_Sparkles {
  simTarget=GPUComputeSim
  emitterUpdateScript [
   &EmitterState(Life Cycle Mode=::NewEnumerator1, Loop Behavior=::NewEnumerator1, Loop Delay=0.1, Loop Duration=1.1, UseLoopDelay=false)
   &SpawnBurst_Instantaneous(Spawn Count=10, Use Spawn Probability=false, UseLoopCountLimit=false)
   &SpawnRate(SpawnRate=100, Use Spawn Probability=false)
  ]
  particleSpawnScript [
   &InitializeParticle(
    Color=<rgba 3.9583 5.8997 10 1>
    Lifetime=&RandomRangeFloat(Minimum=0.3, Maximum=0.5)
    Sprite Rotation Mode=::NewEnumerator0
    Sprite Size Mode=::NewEnumerator1
    UsePositionOffset=false
    Write Material Random=true
   )
   &BoxLocation(Box Offset=<0 0 50>, Box Size=<0 0 100>, MaskSpawnGroup=false, Module.Override Seed=false, Rotate Box=false, Surface Only=false)
   &ShapeLocation(Handle Radius=20, Large Radius=30, MaskSpawnGroup=false, Override Seed=false, Shape Primitive=::NewEnumerator3, UseOwnerScale=true)
   &AddVelocity(
    Cone Angle=5
    Cone Axis=<0 0 1>
    Module.UseConeFalloff=false
    Module.UseVelDistribution=false
    Override Seed=false
    Velocity Mode=::NewEnumerator2
    Velocity Speed=&RandomRangeFloat(Minimum=400, Maximum=500)
   )
   &CameraOffset(Camera Offset Amount=20)
  ]
  particleUpdateScript [
   /SetVariables(Particles.Glow03_Color=&MakeLinearColorFromVectorAndFloat(
     "Vector (RGB)"=&MakeVectorFromLinearColorRGB(Color=&MultiplyLinearColor(
       Linear Color=<rgba 0 0.2079 1 0.2>
       Scale Factor=&MakeVector4_FromFloat(Float=&FNTimeOfDayScalar(Night Scale=0.5, Morning Scale=0.2, Evening Scale=0.2))
      ))
     "Float (Alpha)"=&Multiply_Float(A=0.2, B=&FN_FrontEndFloatModifier(Gameplay=1, Front End=2))
    )) {isSetParameters=true, module=SetVariables_#2}
   /SetVariables(Particles.Glow03_Size=&Multiply_Vector2D_ByFloat(
     Float=&FloatFromCurve(
      FloatCurve=DataInterface() {propertyValues {Curve {keys [
          <at 0 0 RCIM_Cubic>
          <at 0.2 1.3 RCIM_Cubic>
          {arriveTangent=-0.1423, interpMode=RCIM_Cubic, leaveTangent=-0.1423, tangentMode=RCTM_User, time=1, value=0}
         ]}}}
      Scale Curve=20
     )
     Vector2D=<1.5 1.5>
    )) {isSetParameters=true, module=SetVariables_#3}
   &ParticleState()
   &CurlNoiseForce(Noise Strength=&Multiply_Float(A=200, B=1), Pan Noise=true, Pan Noise Field=<0 0 1>)
   &ScaleSpriteSize(Scale Factor=&Vector2DFromFloat(Value=&FloatFromCurve(
      FloatCurve=DataInterface() {propertyValues {Curve {keys [
          <at 0 0 RCIM_Cubic>
          <at 0.2 1.2 RCIM_Cubic>
          {arriveTangent=-0.705, interpMode=RCIM_Cubic, leaveTangent=-0.705, tangentMode=RCTM_User, time=1, value=0}
         ]}}}
      Scale Curve=0.5
     )))
   &ScaleColor(
    Scale Alpha=&Multiply_Float(A=1, B=1)
    Scale RGB=&Multiply_VectorByFloat(Vector=<2 2 2>, Float=&FNTimeOfDayScalar(Night Scale=1, Morning Scale=0.7, Evening Scale=0.7))
    ScaleA=true
    ScaleRGB=true
   )
   &Drag(Drag=3, Use Linear Drag=true, Use Rotational Drag=false)
   &SolveForcesAndVelocity(Clamp Velocity=false, Limit Acceleration=false)
  ]
  rendererClasses [&NiagaraSpriteRendererProperties]
  renderers [
   {rendererClass=&NiagaraSpriteRendererProperties}
   {rendererClass=&NiagaraSpriteRendererProperties, rendererIndex=1}
  ]
 }
 Explode_Sparkles001 {
  simTarget=GPUComputeSim
  emitterUpdateScript [
   &EmitterState(
    Inactive Response=::NewEnumerator1
    Life Cycle Mode=::NewEnumerator1
    Loop Behavior=::NewEnumerator1
    Loop Delay=0.1
    Loop Duration=1
    UseLoopDelay=false
   )
   &SpawnBurst_Instantaneous(Spawn Count=60, Use Spawn Probability=false, UseLoopCountLimit=false)
  ]
  particleSpawnScript [
   &InitializeParticle(
    Color=&MultiplyLinearColor(
     Linear Color=&HueShiftLinearColor(
      Hue=0
      Saturation=0.8
      Value=1
      Linear Color=&ColorFromCurve(
       ColorCurve=DataInterface() {propertyValues {
         AlphaCurve {keys [<at 0 1 RCIM_Cubic>]}
         BlueCurve {keys [<at 0 0 RCIM_Cubic>,<at 0.5 0>,<at 1 1 RCIM_Cubic>]}
         GreenCurve {keys [<at 0 0 RCIM_Cubic>,<at 0.5 1>,<at 1 0 RCIM_Cubic>]}
         RedCurve {keys [<at 0 1 RCIM_Cubic>,<at 0.5 0>,<at 1 0 RCIM_Cubic>]}
        }}
       CurveIndex=&RandomRangeFloat(Minimum=0, Maximum=1)
      )
     )
     Scale Factor=&MakeVector4_FromFloat(Float=2)
    )
    Lifetime=&RandomRangeFloat(Minimum=0.3, Maximum=0.8)
    Sprite Rotation Mode=::NewEnumerator0
    Sprite Size Max=<3 20>
    Sprite Size Min=<3 10>
    Sprite Size Mode=::NewEnumerator2
    UsePositionOffset=false
    Write Material Random=true
   )
   /SetVariables(Particles.RandomCore:NiagaraFloat=&FloatFromCurve(
     FloatCurve=DataInterface() {propertyValues {Curve {keys [
         {arriveTangent=0.0559, interpMode=RCIM_Cubic, leaveTangent=0.0559, tangentMode=RCTM_User, time=0, value=0.2}
         {arriveTangent=0.1895, interpMode=RCIM_Cubic, leaveTangent=0.1895, tangentMode=RCTM_User, time=0.9, value=0.25}
         {arriveTangent=1.1808, interpMode=RCIM_Cubic, leaveTangent=1.1808, tangentMode=RCTM_User, time=1, value=0.3}
        ]}}}
     CurveIndex=&RandomRangeFloat(Minimum=0, Maximum=1)
     Scale Curve=2
    )) {isSetParameters=true}
   &BoxLocation(Box Offset=<0 0 120>, Box Size=<0 0 0>, MaskSpawnGroup=false, Module.Override Seed=false, Rotate Box=false, Surface Only=false)
   &ShapeLocation(MaskSpawnGroup=false, Override Seed=false, Sphere Radius=20, UseOwnerScale=true)
   &AddVelocity(
    Module.Use Falloff Distance=false
    Override Seed=false
    Velocity Mode=::NewEnumerator1
    Velocity Origin=$Particles.Position
    Velocity Speed=&RandomRangeFloat(Minimum=400, Maximum=400)
   )
  ]
  particleUpdateScript [
   /SetVariables(Particles.Glow03_Color=&MakeLinearColorFromVectorAndFloat(
     "Vector (RGB)"=&MakeVectorFromLinearColorRGB(Color=&MultiplyLinearColor(
       Linear Color=&MultiplyLinearColor(Linear Color=$Particles.Initial.Color, Scale Factor=<1 1 1 0.2>)
       Scale Factor=&MakeVector4_FromFloat(Float=&FNTimeOfDayScalar(Night Scale=0.7, Morning Scale=0.2, Evening Scale=0.2))
      ))
     "Float (Alpha)"=&Multiply_Float(A=0.2, B=1)
    )) {isSetParameters=true, module=SetVariables_#2}
   /SetVariables(Particles.Glow03_Size=&Multiply_Vector2D_ByFloat(
     Float=&FloatFromCurve(
      FloatCurve=DataInterface() {propertyValues {Curve {keys [
          <at 0 0 RCIM_Cubic>
          <at 0.2 1.3 RCIM_Cubic>
          {arriveTangent=-0.1423, interpMode=RCIM_Cubic, leaveTangent=-0.1423, tangentMode=RCTM_User, time=1, value=0}
         ]}}}
      Scale Curve=20
     )
     Vector2D=<1 2>
    )) {isSetParameters=true, module=SetVariables_#3}
   &ParticleState()
   &CurlNoiseForce(Noise Strength=&Multiply_Float(A=2000, B=1), Pan Noise=true, Pan Noise Field=<0 0 1>)
   &ScaleSpriteSize(Scale Factor=&Vector2DFromFloat(Value=&FloatFromCurve(
      FloatCurve=DataInterface() {propertyValues {Curve {keys [
          <at 0 0 RCIM_Cubic>
          <at 0.2 1.2 RCIM_Cubic>
          {arriveTangent=-2.7132, interpMode=RCIM_Cubic, leaveTangent=-2.7132, tangentMode=RCTM_User, time=1, value=0}
         ]}}}
      Scale Curve=0.5
     )))
   &ScaleColor(
    Scale Alpha=&Multiply_Float(A=1, B=1)
    Scale RGB=&Multiply_VectorByFloat(Vector=<2 2 2>, Float=&FNTimeOfDayScalar(Night Scale=1, Morning Scale=0.7, Evening Scale=0.7))
    ScaleA=true
    ScaleRGB=true
   )
   &Drag(Drag=2, Use Linear Drag=true, Use Rotational Drag=false)
   &SolveForcesAndVelocity(Clamp Velocity=false, Limit Acceleration=false)
  ]
  rendererClasses [&NiagaraSpriteRendererProperties]
  renderers [
   {rendererClass=&NiagaraSpriteRendererProperties}
   {rendererClass=&NiagaraSpriteRendererProperties, rendererIndex=1}
  ]
 }
 Flare {
  emitterUpdateScript [
   &EmitterState(Life Cycle Mode=::NewEnumerator1, Loop Behavior=::NewEnumerator1, Loop Duration=1, UseLoopDelay=false)
   &SpawnBurst_Instantaneous(Spawn Count=1, Use Spawn Probability=false)
  ]
  particleSpawnScript [
   &InitializeParticle(
    Color=<rgba 2 1.0873 0 1>
    Lifetime=0.4
    Sprite Rotation Mode=::NewEnumerator2
    Sprite Size Mode=::NewEnumerator3
    Uniform Sprite Size=600
    UsePositionOffset=false
    Write Material Random=true
   )
   &DynamicMaterialParameters(
    Index 0 Param 1=&RandomRangeFloat(Minimum=0.4, Maximum=0.6)
    Index 0 Param 2=1
    Index 0 Param 3=1
    Index 0 Param 4=1
    Module.Param0WriteEnabled=true
    Module.Param1WriteEnabled=true
    Module.Param2WriteEnabled=true
    Module.Param3WriteEnabled=true
   )
   &SystemLocation(MaskSpawnGroup=false, Offset=<0 0 150>)
   /SetVariables() {isSetParameters=true}
   /SetVariables() {isSetParameters=true}
   &AddVelocity(Velocity=<0 0 -100>)
  ]
  particleUpdateScript [
   &ParticleState()
   &Drag(Drag=2, Use Linear Drag=true, Use Rotational Drag=false)
   &SolveForcesAndVelocity(Clamp Velocity=false, Limit Acceleration=false)
   &ScaleColor(
    Scale Alpha=&Multiply_Float(
     A=&FNTimeOfDayScalar(Night Scale=0.4, Morning Scale=0.3, Evening Scale=0.4)
     B=&FloatFromCurve(FloatCurve=DataInterface() {propertyValues {Curve {keys [<at 0 0>,<at 0.2 1>,<at 0.8 1>,<at 1 0>]}}})
    )
    ScaleA=true
    ScaleRGB=true
   )
   &ScaleSpriteSize(Scale Factor=&Vector2DFromFloat(Value=&FloatFromCurve(FloatCurve=DataInterface() {propertyValues {Curve {keys [<at 0 1.5 RCIM_Cubic>,<at 1 0>]}}})))
   &CameraOffset(Camera Offset Amount=50)
   /SetVariables(Particles.Flash size:Vector2f=&Multiply_Vector2D_ByFloat(
     Float=&Multiply_Float(
      A=400
      B=&FloatFromCurve(FloatCurve=DataInterface() {propertyValues {Curve {keys [
           {arriveTangent=4.1785, interpMode=RCIM_Cubic, leaveTangent=4.1785, tangentMode=RCTM_User, time=0, value=0}
           {arriveTangent=0.9995, interpMode=RCIM_Cubic, leaveTangent=0.9995, tangentMode=RCTM_User, time=0.1378, value=0.6825}
           {arriveTangent=-0.0405, interpMode=RCIM_Cubic, leaveTangent=-0.0405, tangentMode=RCTM_User, time=0.62, value=1}
          ]}}})
     )
     Vector2D=<1 1>
    )) {isSetParameters=true}
   /SetVariables(Particles.Flash color:LinearColor=&MakeLinearColorFromVectorAndFloat(
     "Vector (RGB)"=&MakeVectorFromLinearColorRGB(Color=<rgba 1 0.5585 0 1>)
     "Float (Alpha)"=&Multiply_Float(
      A=&Multiply_Float(
       A=&FNTimeOfDayScalar(Night Scale=0.2, Morning Scale=0.3, Evening Scale=0.2)
       B=&FloatFromCurve(FloatCurve=DataInterface() {propertyValues {Curve {keys [<at 0 0>,<at 0.124 1>,<at 0.496 1>,<at 0.62 0>]}}})
      )
      B=1
     )
    )) {isSetParameters=true}
   /SetVariables(
    Particles.Glow01Color:LinearColor=&MakeLinearColorFromVectorAndFloat(
     "Vector (RGB)"=&MakeVectorFromLinearColorRGB(Color=<rgba 0.7 0.5798 0.1011 1>)
     "Float (Alpha)"=&FloatFromCurve(
      FloatCurve=DataInterface() {propertyValues {Curve {keys [
          {interpMode=RCIM_Cubic, tangentMode=RCTM_User, time=0, value=0}
          {interpMode=RCIM_Cubic, tangentMode=RCTM_User, time=0.25, value=1}
          {tangentMode=RCTM_User, time=0.5, value=0}
         ]}}}
      Scale Curve=&Multiply_Float(
       A=&Multiply_Float(
        A=&FNTimeOfDayScalar(Night Scale=0.01, Morning Scale=0.3, Evening Scale=0.01)
        B=&FloatFromCurve(FloatCurve=DataInterface() {propertyValues {Curve {keys [<at 0 0>,<at 0.1 1>,<at 0.4 1>,<at 0.5 0>]}}})
       )
       B=1
      )
     )
    )
    Particles.Glow01Size:Vector2f=&Vector2DFromFloat(Value=&FloatFromCurve(FloatCurve=DataInterface() {propertyValues {Curve {keys [<at 0 0.5 RCIM_Cubic>,<at 0.25 1 RCIM_Cubic>,<at 0.5 0.5>]}}}, Scale Curve=600))
   ) {isSetParameters=true}
   /SetVariables(
    Particles.Glow02Color:LinearColor=&MultiplyLinearColor(
     Linear Color=<rgba 1 0.5288 0.2147 1>
     Scale Factor=&MakeFloat4FromVectorAndFloat(
      "Vector (RGB)"=<1 1 1>
      "Float (Alpha)"=&Multiply_Float(
       A=&FloatFromCurve(
        FloatCurve=DataInterface() {propertyValues {Curve {keys [
            {interpMode=RCIM_Cubic, tangentMode=RCTM_User, time=0, value=0}
            {interpMode=RCIM_Cubic, tangentMode=RCTM_User, time=0.2, value=1}
            {tangentMode=RCTM_User, time=0.5, value=0}
           ]}}}
        Scale Curve=0.1
       )
       B=1
      )
     )
    )
    Particles.Glow02Size:Vector2f=&Multiply_Vector2D_ByFloat(
     Float=&FloatFromCurve(FloatCurve=DataInterface() {propertyValues {Curve {keys [
          {interpMode=RCIM_Cubic, tangentMode=RCTM_User, time=0, value=0.5}
          {interpMode=RCIM_Cubic, tangentMode=RCTM_User, time=0.2, value=1}
          {tangentMode=RCTM_User, time=0.5, value=0.5}
         ]}}})
     Vector2D=<600 150>
    )
   ) {isSetParameters=true}
  ]
  rendererClasses [&NiagaraSpriteRendererProperties]
  renderers [
   {rendererClass=&NiagaraSpriteRendererProperties}
   {rendererClass=&NiagaraSpriteRendererProperties, rendererIndex=1}
   {rendererClass=&NiagaraSpriteRendererProperties, rendererIndex=2}
   {rendererClass=&NiagaraSpriteRendererProperties, rendererIndex=3}
  ]
 }
 Flare_01001 {
  emitterUpdateScript [
   &EmitterState()
   &SpawnBurst_Instantaneous(Use Spawn Probability=false, UseLoopCountLimit=false)
  ]
  particleSpawnScript [
   &InitializeParticle(
    Color=<rgba 1 0.9464 0.8 0.5>
    Lifetime=$System.CurrentLoopDuration
    Sprite Size=&Multiply_Vector2D_ByFloat(Float=&Multiply_Float(A=1, B=1.2), Vector2D=<180 250>)
    Sprite Size Mode=::NewEnumerator4
    UsePositionOffset=false
    Write Material Random=true
   )
   &SystemLocation(MaskSpawnGroup=false, Offset=<0 0 100>)
   &CameraOffset(Camera Offset Amount=-50)
  ]
  particleUpdateScript [
   &ParticleState()
   &SpriteRotationRate(Rotation Rate=6)
   &ScaleColor(
    Scale Alpha=&FloatFromCurve(
     FloatCurve=DataInterface() {propertyValues {Curve {keys [
         {arriveTangent=25.4717, interpMode=RCIM_Cubic, leaveTangent=25.4717, tangentMode=RCTM_User, time=0, value=0}
         {arriveTangent=0.1236, interpMode=RCIM_Cubic, leaveTangent=0.1236, tangentMode=RCTM_User, time=0.1, value=1}
         <at 0.5 1 RCIM_Cubic>
         {arriveTangent=-0.4595, interpMode=RCIM_Cubic, leaveTangent=-0.4595, tangentMode=RCTM_User, time=0.8, value=1}
         {arriveTangent=-26.3419, interpMode=RCIM_Cubic, leaveTangent=-26.3419, tangentMode=RCTM_User, time=0.9, value=0}
        ]}}}
     CurveIndex=$System.NormalizedLoopAge
    )
    Scale RGB=&Multiply_VectorByFloat(
     Vector=&VectorFromFloat(Value=&FNTimeOfDayScalar(Night Scale=0.5, Morning Scale=0.25, Evening Scale=0.5))
     Float=&FN_FrontEndFloatModifier(Gameplay=1, Front End=1)
    )
    ScaleA=true
    ScaleRGB=true
   )
   &ScaleSpriteSize(Uniform Curve Sprite Scale=DataInterface() {propertyValues {Curve {keys [
        {arriveTangent=26.1571, interpMode=RCIM_Cubic, leaveTangent=26.1571, tangentMode=RCTM_User, time=0, value=0}
        {arriveTangent=1.3333, interpMode=RCIM_Cubic, leaveTangent=1.3333, time=0.08, value=0.9}
        {arriveTangent=-0.9783, interpMode=RCIM_Cubic, leaveTangent=-0.9783, time=0.75, value=1}
        <at 1 0 RCIM_Cubic>
       ]}}})
   /SetVariables(
    Particles.Color_Scale:LinearColor=&MakeLinearColorFromVectorAndFloat(
     "Vector (RGB)"=&Multiply_Vector(
      A=&MakeVectorFromLinearColorRGB(Color=<rgba 1.8806 2 1.6 0.2>)
      B=&Multiply_VectorByFloat(
       Vector=&VectorFromFloat(Value=&FNTimeOfDayScalar(Night Scale=0.5, Morning Scale=0.25, Evening Scale=0.5))
       Float=&FN_FrontEndFloatModifier(Gameplay=1, Front End=1)
      )
     )
     "Float (Alpha)"=&FloatFromCurve(
      FloatCurve=DataInterface() {propertyValues {Curve {keys [
          {arriveTangent=25.4717, interpMode=RCIM_Cubic, leaveTangent=25.4717, tangentMode=RCTM_User, time=0, value=0}
          {arriveTangent=0.1236, interpMode=RCIM_Cubic, leaveTangent=0.1236, tangentMode=RCTM_User, time=0.1, value=1}
          <at 0.5 1 RCIM_Cubic>
          {arriveTangent=-0.4595, interpMode=RCIM_Cubic, leaveTangent=-0.4595, tangentMode=RCTM_User, time=0.8, value=1}
          {arriveTangent=-26.3419, interpMode=RCIM_Cubic, leaveTangent=-26.3419, tangentMode=RCTM_User, time=0.9, value=0}
         ]}}}
      CurveIndex=$System.NormalizedLoopAge
     )
    )
    Particles.Sprite_Size:Vector2f=&Multiply_Vector2D_ByFloat(
     Float=&Multiply_Float(
      A=&FloatFromCurve(FloatCurve=DataInterface() {propertyValues {Curve {keys [
           {arriveTangent=27.5356, interpMode=RCIM_Cubic, leaveTangent=27.5356, tangentMode=RCTM_User, time=0, value=0}
           {arriveTangent=-0.0043, interpMode=RCIM_Cubic, leaveTangent=-0.0043, tangentMode=RCTM_User, time=0.08, value=1}
           {arriveTangent=-0.1074, interpMode=RCIM_Cubic, leaveTangent=-0.1074, tangentMode=RCTM_User, time=0.9, value=1}
           {arriveTangent=-21.0911, interpMode=RCIM_Cubic, leaveTangent=-21.0911, tangentMode=RCTM_User, time=1, value=0}
          ]}}})
      B=&RandomRangeFloat(Minimum=1.2, Maximum=1.5)
     )
     Vector2D=<200 180>
    )
   ) {isSetParameters=true}
   /SetVariables(
    Particles.Color01_Scale:LinearColor=&MakeLinearColorFromVectorAndFloat(
     "Vector (RGB)"=&MakeVectorFromLinearColorRGB(Color=<rgba 0.6042 0.3987 0.2417 0.7>)
     "Float (Alpha)"=&FloatFromCurve(
      FloatCurve=DataInterface() {propertyValues {Curve {keys [
          {arriveTangent=25.4717, interpMode=RCIM_Cubic, leaveTangent=25.4717, tangentMode=RCTM_User, time=0, value=0}
          {arriveTangent=0.1236, interpMode=RCIM_Cubic, leaveTangent=0.1236, tangentMode=RCTM_User, time=0.1, value=1}
          <at 0.5 1 RCIM_Cubic>
          {arriveTangent=-0.4595, interpMode=RCIM_Cubic, leaveTangent=-0.4595, tangentMode=RCTM_User, time=0.8, value=1}
          {arriveTangent=-26.3419, interpMode=RCIM_Cubic, leaveTangent=-26.3419, tangentMode=RCTM_User, time=0.9, value=0}
         ]}}}
      CurveIndex=$System.NormalizedLoopAge
     )
    )
    Particles.Sprite01_Size:Vector2f=<500 500>
   ) {isSetParameters=true}
   /SetVariables(
    Particles.CameraOffset_BackGlow=-20
    Particles.Color02_Scale:LinearColor=&MakeLinearColorFromVectorAndFloat(
     "Vector (RGB)"=&MakeVectorFromLinearColorRGB(Color=<rgba 0.0521 0.0409 0.0271 0.5>)
     "Float (Alpha)"=&FloatFromCurve(
      FloatCurve=DataInterface() {propertyValues {Curve {keys [
          {arriveTangent=25.4717, interpMode=RCIM_Cubic, leaveTangent=25.4717, tangentMode=RCTM_User, time=0, value=0}
          {arriveTangent=0.1236, interpMode=RCIM_Cubic, leaveTangent=0.1236, tangentMode=RCTM_User, time=0.1, value=1}
          <at 0.5 1 RCIM_Cubic>
          {arriveTangent=-0.4595, interpMode=RCIM_Cubic, leaveTangent=-0.4595, tangentMode=RCTM_User, time=0.8, value=1}
          {arriveTangent=-26.3419, interpMode=RCIM_Cubic, leaveTangent=-26.3419, tangentMode=RCTM_User, time=0.9, value=0}
         ]}}}
      CurveIndex=$System.NormalizedLoopAge
      Scale Curve=1.5
     )
    )
   ) {isSetParameters=true}
   &DynamicMaterialParameters(
    Index 0 Param 1=&FloatFromCurve(FloatCurve=DataInterface() {propertyValues {Curve {keys [<at 0 0.5>,<at 1 0>]}}})
    Index 0 Param 2=2
    Module.Param0WriteEnabled=true
    Module.Param1WriteEnabled=true
    Module.Param2WriteEnabled=true
    Module.Param3WriteEnabled=true
   )
  ]
  rendererClasses [&NiagaraSpriteRendererProperties]
  renderers [
   {rendererClass=&NiagaraSpriteRendererProperties}
   {rendererClass=&NiagaraSpriteRendererProperties, rendererIndex=1}
   {rendererClass=&NiagaraSpriteRendererProperties, rendererIndex=2}
   {rendererClass=&NiagaraSpriteRendererProperties, rendererIndex=3}
  ]
 }
 Flare_03001 {
  emitterUpdateScript [
   &EmitterState()
   &SpawnBurst_Instantaneous(Use Spawn Probability=false, UseLoopCountLimit=false)
   &SpawnRate(SpawnRate=8, Use Spawn Probability=false)
   &SpawnBurst_Instantaneous(Spawn Time=0.133, Use Spawn Probability=false, UseLoopCountLimit=false) {module=SpawnBurst_Instantaneous001}
  ]
  particleSpawnScript [
   &InitializeParticle(
    Color=<rgba 1 0.8888 0.4989 0.25>
    Lifetime Max=0.4
    Lifetime Min=0.2
    Lifetime Mode=::NewEnumerator1
    Sprite Size=&Multiply_Vector2D_ByFloat(
     Float=&Multiply_Float(A=1, B=&RandomRangeFloat(Minimum=1.2, Maximum=1.5))
     Vector2D=<250 180>
    )
    Sprite Size Mode=::NewEnumerator4
    UsePositionOffset=false
    Write Material Random=true
   )
   &SystemLocation(MaskSpawnGroup=false, Offset=<0 0 100>)
   &CameraOffset(Camera Offset Amount=-80)
  ]
  particleUpdateScript [
   &ParticleState()
   &ScaleColor(
    Scale Alpha=&Multiply_Float(
     A=&FloatFromCurve(
      FloatCurve=DataInterface() {propertyValues {Curve {keys [
          {interpMode=RCIM_Cubic, tangentMode=RCTM_User, time=0, value=0}
          {interpMode=RCIM_Cubic, tangentMode=RCTM_User, time=0.5, value=1}
          {tangentMode=RCTM_User, time=1, value=0}
         ]}}}
      Scale Curve=&FN_FrontEndFloatModifier(Gameplay=1, Front End=0.5)
     )
     B=&FloatFromCurve(
      FloatCurve=DataInterface() {propertyValues {Curve {keys [
          {arriveTangent=25.4717, interpMode=RCIM_Cubic, leaveTangent=25.4717, tangentMode=RCTM_User, time=0, value=0}
          {arriveTangent=0.1236, interpMode=RCIM_Cubic, leaveTangent=0.1236, tangentMode=RCTM_User, time=0.1, value=1}
          <at 0.5 1 RCIM_Cubic>
          {arriveTangent=-0.4595, interpMode=RCIM_Cubic, leaveTangent=-0.4595, tangentMode=RCTM_User, time=0.8, value=1}
          {arriveTangent=-26.3419, interpMode=RCIM_Cubic, leaveTangent=-26.3419, tangentMode=RCTM_User, time=0.9, value=0}
         ]}}}
      CurveIndex=$System.NormalizedLoopAge
     )
    )
    Scale RGB=&Multiply_VectorByFloat(
     Vector=&VectorFromFloat(Value=&FNTimeOfDayScalar(Night Scale=0.5, Morning Scale=0.25, Evening Scale=0.5))
     Float=&FN_FrontEndFloatModifier(Gameplay=1, Front End=1)
    )
    ScaleA=true
    ScaleRGB=true
   )
   &ScaleSpriteSize(Uniform Curve Sprite Scale=DataInterface() {propertyValues {Curve {keys [{tangentMode=RCTM_User, time=0, value=0}, <at 0.1 0.7>, {tangentMode=RCTM_User, time=0.9, value=1}, <at 1 0>]}}})
   &DynamicMaterialParameters(
    Index 0 Param 1=&RandomRangeFloat(Minimum=-1, Maximum=1)
    Index 0 Param 4=&RandomRangeFloat(Minimum=0, Maximum=1)
    Module.Param0WriteEnabled=true
    Module.Param1WriteEnabled=true
    Module.Param2WriteEnabled=true
    Module.Param3WriteEnabled=true
   )
  ]
  rendererClasses [&NiagaraSpriteRendererProperties]
  renderers [{rendererClass=&NiagaraSpriteRendererProperties}]
 }
 HA {
  emitterSpawnScript [/SetVariables(Emitter.DelayHA=0.5) {isSetParameters=true}]
  emitterUpdateScript [
   &EmitterState(
    Enable Distance Culling=true
    Enable Visibility Culling=true
    MaxDistance=1000
    Module.bApplySpawnCountScaleByDistance=true
    Module.bUseMaxDistance=true
    Module.bUseMinDistance=false
    Scalability Mode=::NewEnumerator1
    Scale Spawn Count=true
    Spawn Count Scale By Distance=DataInterface() {propertyValues {Curve {keys [<at 0 1>,<at 0.3 1>,<at 1 0>]}}}
   )
   &SpawnBurst_Instantaneous(Spawn Time=&Add_Float(A=$Emitter.Initial.DelayHA, B=0), Use Spawn Probability=false, UseLoopCountLimit=false)
   &SpawnBurst_Instantaneous(Spawn Time=&Add_Float(A=$Emitter.Initial.DelayHA, B=0.3), Use Spawn Probability=false, UseLoopCountLimit=false) {module=SpawnBurst_Instantaneous001}
   &SpawnBurst_Instantaneous(Spawn Time=&Add_Float(A=$Emitter.Initial.DelayHA, B=0.6), Use Spawn Probability=false, UseLoopCountLimit=false) {module=SpawnBurst_Instantaneous002}
   &SpawnBurst_Instantaneous(Spawn Time=&Add_Float(A=$Emitter.Initial.DelayHA, B=0.9), Use Spawn Probability=false, UseLoopCountLimit=false) {module=SpawnBurst_Instantaneous003}
   &SpawnRate(SpawnRate=80, Use Spawn Probability=false) {enabled=false, module=SpawnRate001}
  ]
  particleSpawnScript [
   &PartitionParticles(Number Of Partitions=4)
   &InitializeParticle(
    Lifetime=3
    Position Offset=&SelectVectorFromArray(
     Direct Array Index=$Particles.PartitionParticles.Partition
     Vector Selection Array=DataInterface() {propertyValues {FloatData [<40 0 180>, <10 0 190>, <-15 0 185>, <-40 0 177>]}}
    )
    Position Offset Coordinate Space=::Local
    Sprite Rotation Angle=&SelectFloatFromArray(
     Direct Array Index=$Particles.PartitionParticles.Partition
     Float Selection Array=DataInterface() {propertyValues {FloatData [-0.04,0.08,0.02,0.1]}}
    )
    Sprite Rotation Mode=::NewEnumerator1
    Sprite Size Mode=::NewEnumerator3
    Uniform Sprite Size=60
    UsePositionOffset=true
    Write Material Random=true
   ) {module=InitializeParticle001}
   &AddVelocity(Velocity=<0 -2 0>)
   &InitialMeshOrientation(
    Mesh Axis To Orient=<0 1 0>
    Mesh Orientation Mode=::NewEnumerator2
    Orientation Coordinate Space=::Local
    Orientation Vector=<0 0 1>
    Use Rotation Vector=true
   )
   &ApplyInitialForces()
   &AlignSpriteToMeshOrientation(SetAlignment=true, SetFacing=true)
  ]
  particleUpdateScript [
   &ParticleState()
   &ScaleSpriteSize(
    Uniform Curve Index=$Emitter.Age
    Uniform Curve Sprite Scale=DataInterface() {propertyValues {Curve {keys [
        {tangentMode=RCTM_User, time=0, value=1}
        {arriveTangent=-25.3901, interpMode=RCIM_Cubic, leaveTangent=-25.3901, tangentMode=RCTM_User, time=2.4993, value=0.985}
        {arriveTangent=4.1309, interpMode=RCIM_Cubic, leaveTangent=4.1309, time=2.5634, value=1.3416}
        {arriveTangent=-6.3941, interpMode=RCIM_Cubic, leaveTangent=-6.3941, tangentMode=RCTM_User, time=2.6149, value=1.4626}
        {arriveTangent=-31.8589, interpMode=RCIM_Cubic, leaveTangent=-31.8589, tangentMode=RCTM_User, time=2.6796, value=-0.0075}
       ]}}}
   ) {module=ScaleSpriteSize001}
   &ScaleSpriteSize(
    Scale Sprite Size Mode=::NewEnumerator0
    Uniform Scale Factor=&FloatFromCurve(
     FloatCurve=DataInterface() {propertyValues {Curve {keys [
         {arriveTangent=9.0725, interpMode=RCIM_Cubic, leaveTangent=9.0725, tangentMode=RCTM_User, time=0, value=0}
         {arriveTangent=-4.8107, interpMode=RCIM_Cubic, leaveTangent=-4.8107, tangentMode=RCTM_User, time=0.07, value=0.9847}
         {interpMode=RCIM_Cubic, tangentMode=RCTM_User, time=0.12, value=0.6}
         {arriveTangent=-0.0405, interpMode=RCIM_Cubic, leaveTangent=-0.0405, tangentMode=RCTM_User, time=1, value=0.8}
        ]}}}
     Scale Curve=0.7
    )
   )
   &SolveForcesAndVelocity(Clamp Velocity=false, Limit Acceleration=false)
  ]
  rendererClasses [&NiagaraSpriteRendererProperties]
  renderers [{rendererClass=&NiagaraSpriteRendererProperties}]
 }
 HQ_Light {
  bEnabled=false
  emitterUpdateScript [
   &EmitterState(Enable Distance Culling=true, MaxDistance=1500, Module.bUseMaxDistance=true, Module.bUseMinDistance=false, Scalability Mode=::NewEnumerator1)
   &SpawnBurst_Instantaneous(Spawn Count=1, Spawn Time=0.133, Use Spawn Probability=false)
  ]
  particleSpawnScript [&InitializeParticle(Color Mode=::NewEnumerator0, Lifetime=$System.CurrentLoopDuration, UsePositionOffset=false, Write Material Random=true)]
  particleUpdateScript [
   &ParticleState()
   &FN_Light_Attributes(
    Color=&MakeLinearColorFromVectorAndFloat(
     "Vector (RGB)"=&FN_FrontEndVector3Modifier(Gameplay=<2 1.6 1.4>, Front End=<0.5 0.1 0>)
     "Float (Alpha)"=1
    )
    Create Color Attribute=true
    Create Position Attribute=true
    Create Radius Attribute=true
    Enable Visibility Tag=true
    Position=&Add_Vector(
     A=Linked() {linkedVariable {name=Particles.Position, type=NiagaraPosition}}
     B=&TransformVector(Vector=<0 0 50>, Source Space=::World, Destination Space=::Local)
    )
    Radius=250
    Scale Color=&FloatFromCurve(
     FloatCurve=DataInterface() {propertyValues {Curve {keys [
         {arriveTangent=117.3003, interpMode=RCIM_Cubic, leaveTangent=117.3003, tangentMode=RCTM_User, time=0, value=0}
         {arriveTangent=1.3594, interpMode=RCIM_Cubic, leaveTangent=1.3594, tangentMode=RCTM_User, time=0.0286, value=1}
         {arriveTangent=-1.0294, interpMode=RCIM_Cubic, leaveTangent=-1.0294, time=0.8, value=0.9805}
         {arriveTangent=-3.2683, interpMode=RCIM_Cubic, leaveTangent=-3.2683, tangentMode=RCTM_User, time=1, value=0}
        ]}}}
     Scale Curve=3
    )
    VisibilityTag=HlslExpression() {hlslExpression="Particles.UniqueID && 1"}
   )
  ]
  rendererClasses [&NiagaraLightRendererProperties]
  renderers [{rendererClass=&NiagaraLightRendererProperties, rendererIndex=0}]
 }
 InkPop {
  emitterUpdateScript [
   &EmitterState(
    Enable Distance Culling=true
    Enable Visibility Culling=true
    MaxDistance=1000
    Module.bApplySpawnCountScaleByDistance=true
    Module.bUseMaxDistance=true
    Module.bUseMinDistance=false
    Scalability Mode=::NewEnumerator1
    Scale Spawn Count=true
    Spawn Count Scale By Distance=DataInterface() {propertyValues {Curve {keys [<at 0 1>,<at 0.3 1>,<at 1 0>]}}}
   )
   &SpawnBurst_Instantaneous(Spawn Count=20, Use Spawn Probability=false, UseLoopCountLimit=false)
   &SpawnRate(SpawnRate=80, Use Spawn Probability=false) {enabled=false, module=SpawnRate001}
  ]
  particleSpawnScript [
   &PartitionParticles(Number Of Partitions=4)
   &InitializeParticle(
    Color=<rgba 0 0 0 1>
    Lifetime=&Multiply_Float(A=&RandomRangeFloat(Minimum=0.15, Maximum=0.25), B=1)
    Sprite Size Mode=::NewEnumerator1
    Uniform Sprite Size Max=3
    Uniform Sprite Size Min=1
    UsePositionOffset=false
    Write Material Random=true
   ) {module=InitializeParticle001}
   &SkeletalMeshLocation(
    Bone / Socket Index=$Particles.PartitionParticles.Partition
    Bone Sampling Mode=::NewEnumerator4
    MaskSpawnGroup=false
    Skeletal Mesh=$User.SkeletalMesh
   )
   &ShapeLocation(
    Additional Quaternion Rotation=true
    Additional Yaw / Pitch / Roll=&Lerp_Vector(
     A=<0 0 0>
     B=<0 0.5 0>
     Alpha=&ModuloFloat(A=&MakeFloatFromInt(Int=$Particles.PartitionParticles.Partition), B=2)
    )
    Cone Angle=80
    Cone Length=20
    MaskSpawnGroup=false
    Non Uniform Scale=&VectorFromFloat(Value=&Sine(Normalized Angle=$Emitter.Age, Period=1, Scale=0.1, Bias=1))
    Offset Coordinate Space=::Local
    Offset Mode=::NewEnumerator0
    Override Seed=false
    Rotation Mode=::NewEnumerator3
    Rotation Quaternion=$Particles.SkeletalMeshLocation.SampledRotation
    Shape Primitive=::NewEnumerator5
    UseOwnerScale=true
   )
   &AddVelocity(
    Module.Use Falloff Distance=false
    Override Seed=false
    Velocity Mode=::NewEnumerator1
    Velocity Origin=$Particles.SkeletalMeshLocation.SampledPosition
    Velocity Speed=&RandomRangeFloat(Minimum=100, Maximum=200)
   )
  ]
  particleUpdateScript [
   &ParticleState()
   &InheritSourceMovement(
    Applied Position Scale=&VectorFromFloat(Value=&RandomRangeFloat(Minimum=0.95, Maximum=1))
    ApplyForce=false
    ApplyPosition=true
    ApplyPositionFalloff=false
   ) {enabled=false}
   &SolveForcesAndVelocity(Clamp Velocity=false, Limit Acceleration=false)
   &ScaleColor(
    Scale Alpha=&FloatFromCurve(FloatCurve=DataInterface() {propertyValues {Curve {keys [<at 0.5 0.95>,<at 1 0>]}}})
    Scale RGB=&MakeVectorFromLinearColorRGB(Color=&ColorFromCurve(ColorCurve=DataInterface() {propertyValues {
        AlphaCurve {keys [<at 0 1 RCIM_Cubic>,<at 0.7107 1>]}
        BlueCurve {keys [<at 0 14.6701 RCIM_Cubic>,<at 0.2499 1.5104>,<at 0.7114 0.1753 RCIM_Cubic>]}
        GreenCurve {keys [<at 0 36.7481 RCIM_Cubic>,<at 0.2499 5.293>,<at 0.7114 0.3243 RCIM_Cubic>]}
        RedCurve {keys [<at 0 50 RCIM_Cubic>,<at 0.2499 10>,<at 0.7114 1 RCIM_Cubic>]}
       }}))
    ScaleA=true
    ScaleRGB=true
   )
   &ScaleSpriteSize(Scale Factor=&Vector2DFromCurve(
     Vector2Curve=DataInterface() {propertyValues {
       XCurve {keys [
         {arriveTangent=0.9446, interpMode=RCIM_Cubic, leaveTangent=0.9446, tangentMode=RCTM_User, time=0, value=0.625}
         {interpMode=RCIM_Cubic, tangentMode=RCTM_User, time=0.4, value=0.5}
         <at 1 0>
        ]}
       YCurve {keys [
         {arriveTangent=0.9446, interpMode=RCIM_Cubic, leaveTangent=0.9446, tangentMode=RCTM_User, time=0, value=3.5}
         {arriveTangent=-11.1745, interpMode=RCIM_Cubic, leaveTangent=-11.1745, tangentMode=RCTM_User, time=0.4, value=2.25}
         <at 1 0>
        ]}
      }}
     ScaleCurve=&Vector2DFromFloat(Value=&FN_LOD_Distance_Remap(Start Distance=300, End Distance=1000, Minimum Output=1, Maximum Output=1.1))
    ))
   &ScaleSpriteSize(Uniform Curve Scale=2, Uniform Curve Sprite Scale=DataInterface() {propertyValues {Curve {keys [{tangentMode=RCTM_User, time=1, value=1}]}}}) {module=ScaleSpriteSize001}
  ]
  rendererClasses [&NiagaraSpriteRendererProperties]
  renderers [{rendererClass=&NiagaraSpriteRendererProperties}]
 }
 InkPop001 {
  emitterUpdateScript [
   &EmitterState(
    Enable Distance Culling=true
    Enable Visibility Culling=true
    Inactive Response=::NewEnumerator1
    Life Cycle Mode=::NewEnumerator1
    Loop Behavior=::NewEnumerator1
    Loop Delay=2.6
    Loop Duration=3
    MaxDistance=1000
    Module.bApplySpawnCountScaleByDistance=true
    Module.bUseMaxDistance=true
    Module.bUseMinDistance=false
    Scalability Mode=::NewEnumerator1
    Scale Spawn Count=true
    Spawn Count Scale By Distance=DataInterface() {propertyValues {Curve {keys [<at 0 1>,<at 0.3 1>,<at 1 0>]}}}
    UseLoopDelay=true
   )
   &SpawnBurst_Instantaneous(Spawn Count=12, Use Spawn Probability=false, UseLoopCountLimit=false)
   &SpawnRate(SpawnRate=80, Use Spawn Probability=false) {enabled=false, module=SpawnRate001}
  ]
  particleSpawnScript [
   &PartitionParticles(Number Of Partitions=4)
   &InitializeParticle(
    Color=<rgba 0 0 0 1>
    Lifetime=&Multiply_Float(A=&RandomRangeFloat(Minimum=0.2, Maximum=0.3), B=1)
    Position Mode=::NewEnumerator0
    Sprite Size Mode=::NewEnumerator1
    Uniform Sprite Size Max=3
    Uniform Sprite Size Min=1
    Write Material Random=true
   ) {module=InitializeParticle001}
   &SystemLocation(
    MaskSpawnGroup=false
    Offset=&SelectVectorFromArray(
     Direct Array Index=$Particles.PartitionParticles.Partition
     Vector Selection Array=DataInterface() {propertyValues {FloatData [<50 0 170>, <10 0 180>, <-15 0 175>, <-50 0 165>]}}
    )
   )
   &ShapeLocation(
    Cone Angle=80
    Cone Length=20
    MaskSpawnGroup=false
    Non Uniform Scale=&VectorFromFloat(Value=&Sine(Normalized Angle=$Emitter.Age, Period=1, Scale=0.1, Bias=1))
    Offset Coordinate Space=::Local
    Offset Mode=::NewEnumerator0
    Override Seed=false
    Shape Primitive=::NewEnumerator5
    UseOwnerScale=true
   )
   &AddVelocity(
    Module.Use Falloff Distance=false
    Override Seed=false
    Velocity Mode=::NewEnumerator1
    Velocity Origin=$Particles.Position
    Velocity Speed=&RandomRangeFloat(Minimum=100, Maximum=200)
   )
  ]
  particleUpdateScript [
   &ParticleState()
   &GravityForce(Gravity=<0 0 -280>)
   &SolveForcesAndVelocity(Clamp Velocity=false, Limit Acceleration=false)
   &ScaleColor(
    Scale Alpha=&FloatFromCurve(FloatCurve=DataInterface() {propertyValues {Curve {keys [<at 0.5 0.95>,<at 1 0>]}}})
    Scale RGB=&MakeVectorFromLinearColorRGB(Color=&ColorFromCurve(ColorCurve=DataInterface() {propertyValues {
        AlphaCurve {keys [<at 0 1 RCIM_Cubic>,<at 0.7107 1>]}
        BlueCurve {keys [<at 0 14.6701 RCIM_Cubic>,<at 0.2499 1.5104>,<at 0.7114 0.1753 RCIM_Cubic>]}
        GreenCurve {keys [<at 0 36.7481 RCIM_Cubic>,<at 0.2499 5.293>,<at 0.7114 0.3243 RCIM_Cubic>]}
        RedCurve {keys [<at 0 50 RCIM_Cubic>,<at 0.2499 10>,<at 0.7114 1 RCIM_Cubic>]}
       }}))
    ScaleA=true
    ScaleRGB=true
   )
   &ScaleSpriteSize(Scale Factor=&Vector2DFromCurve(
     Vector2Curve=DataInterface() {propertyValues {
       XCurve {keys [
         {arriveTangent=0.9446, interpMode=RCIM_Cubic, leaveTangent=0.9446, tangentMode=RCTM_User, time=0, value=0.625}
         {interpMode=RCIM_Cubic, tangentMode=RCTM_User, time=0.4, value=0.5}
         <at 1 0>
        ]}
       YCurve {keys [
         {arriveTangent=0.9446, interpMode=RCIM_Cubic, leaveTangent=0.9446, tangentMode=RCTM_User, time=0, value=3.5}
         {arriveTangent=-11.1745, interpMode=RCIM_Cubic, leaveTangent=-11.1745, tangentMode=RCTM_User, time=0.4, value=2.25}
         <at 1 0>
        ]}
      }}
     ScaleCurve=&Vector2DFromFloat(Value=&FN_LOD_Distance_Remap(Start Distance=300, End Distance=1000, Minimum Output=1, Maximum Output=1.1))
    ))
   &ScaleSpriteSize(Uniform Curve Scale=2, Uniform Curve Sprite Scale=DataInterface() {propertyValues {Curve {keys [{tangentMode=RCTM_User, time=1, value=1}]}}}) {module=ScaleSpriteSize001}
  ]
  rendererClasses [&NiagaraSpriteRendererProperties]
  renderers [{rendererClass=&NiagaraSpriteRendererProperties}]
 }
 OutwardSparkles {
  simTarget=GPUComputeSim
  emitterUpdateScript [
   &EmitterState()
   &SpawnRate(
    SpawnRate=&FloatFromCurve(
     FloatCurve=DataInterface() {propertyValues {Curve {keys [<at 0 1>,<at 0.6 1>,<at 0.7 0>]}}}
     CurveIndex=$Emitter.NormalizedLoopAge
     Scale Curve=25
    )
    Use Spawn Probability=false
   )
   &SpawnBurst_Instantaneous(Spawn Count=25, Spawn Time=$User.SecondSpawnTiming, Use Spawn Probability=false, UseLoopCountLimit=false)
  ]
  particleSpawnScript [
   &InitializeParticle(
    Color Maximum=<rgba 80 40.9576 16 1>
    Color Minimum=<rgba 80 52.1077 16 1>
    Color Mode=::NewEnumerator2
    Lifetime Max=1
    Lifetime Min=0.1
    Lifetime Mode=::NewEnumerator1
    Mass Max=2
    Mass Mode=::NewEnumerator3
    Position Offset=<0 0 100>
    Sprite Rotation Angle=45
    Sprite Rotation Mode=::NewEnumerator0
    Sprite Size Mode=::NewEnumerator1
    Uniform Sprite Size Max=30
    Uniform Sprite Size Min=25
    UsePositionOffset=true
    Write Material Random=true
   )
   &SphereLocation(
    MaskSpawnGroup=false
    Override Seed=false
    Sphere Radius=50
    Surface Only=true
    Surface Only Band Thickness=0.1
    UseNonUniformScale=false
    UseOffset=false
    UseSphereOrientation=false
   ) {enabled=false}
   &JitterPosition(Jitter Amount=20)
   &AddVelocityFromPoint(Module.Use Falloff Distance=false, Velocity Strength=&RandomRangeFloat(Minimum=150, Maximum=250)) {enabled=false}
   &self:ScratchModule_04(Src:Vector3f=<0 0 64>, Velocity=250) {enabled=false, module=ScratchModule_04}
   &AddVelocity(Velocity=&RandomRangeVector(Minimum=<-1 -1 -1>, Maximum=<1 1 1>), Velocity Speed Scale=500)
  ]
  particleUpdateScript [
   &ParticleState()
   &GravityForce(Gravity=<0 0 0>) {enabled=false}
   &AccelerationForce(Acceleration=&RandomRangeVector(Minimum=<0 0 0>, Maximum=<1 1 100>)) {enabled=false}
   &Drag(Drag=&RandomRangeFloat(Minimum=2.5, Maximum=5), Use Linear Drag=true, Use Rotational Drag=false)
   &ScaleColor(
    Scale Alpha=&FloatFromCurve(FloatCurve=DataInterface() {propertyValues {Curve {keys [<at -0.0046 1.0113 RCIM_Cubic>,<at 1.0011 0.002 RCIM_Cubic>]}}})
    ScaleA=true
    ScaleRGB=true
   )
   &CurlNoiseForce(Noise Frequency=10, Noise Quality / Cost=::NewEnumerator0, Noise Strength=5, Pan Noise=false)
   &SolveForcesAndVelocity(Clamp Velocity=false, Limit Acceleration=false)
   &ScaleSpriteSize(Uniform Curve Sprite Scale=DataInterface() {propertyValues {Curve {keys [
        <at 0 0 RCIM_Cubic>
        {arriveTangent=1, interpMode=RCIM_Cubic, leaveTangent=1, time=0.1, value=0.95}
        {arriveTangent=-1, interpMode=RCIM_Cubic, leaveTangent=-1, time=0.2, value=0.2}
        {arriveTangent=-0.375, interpMode=RCIM_Cubic, leaveTangent=-0.375, time=0.4, value=0.65}
        {arriveTangent=1, interpMode=RCIM_Cubic, leaveTangent=1, time=0.6, value=0.05}
        <at 0.7 0.95 RCIM_Cubic>
        {arriveTangent=0.1667, interpMode=RCIM_Cubic, leaveTangent=0.1667, time=0.9, value=0.05}
        <at 1 1 RCIM_Cubic>
       ]}}})
  ]
  rendererClasses [&NiagaraSpriteRendererProperties]
  renderers [{rendererClass=&NiagaraSpriteRendererProperties}]
 }
 RFlow_Combined {
  emitterUpdateScript [
   &EmitterState(
    Delay First Loop Only=true
    Life Cycle Mode=::NewEnumerator1
    Loop Behavior=::NewEnumerator2
    Loop Count=2
    Loop Delay=0.1
    Loop Duration=0.05
    UseLoopDelay=true
   )
   &SpawnRate(SpawnRate=8, Use Spawn Probability=false)
   &SpawnBurst_Instantaneous(Use Spawn Probability=false, UseLoopCountLimit=false)
  ]
  particleSpawnScript [
   /SetVariables(Particles.MeshIndex=&SelectIntFromArray(
     Direct Array Index=$Emitter.LoopCount
     Direct Array Mode=::NewEnumerator1
     Int Selection Array=DataInterface() {propertyValues {IntData [0,1]}}
    )) {isSetParameters=true}
   &InitializeParticle(
    Color=<rgba 10 6.5333 2 0.1>
    Lifetime=&Multiply_Float(
     A=&RandomRangeFloat(Minimum=0.9, Maximum=1.1)
     B=&SelectFloatFromArray(Direct Array Index=$Particles.MeshIndex, Float Selection Array=DataInterface() {propertyValues {FloatData [0.3,0.6]}})
    )
    Mesh Scale=&RandomRangeVector(Minimum=<1.3 1.3 1>, Maximum=<2 2 1.5>)
    Mesh Scale Mode=::NewEnumerator4
    UsePositionOffset=false
    Write Material Random=true
   )
   &SystemLocation(MaskSpawnGroup=false)
   &InitialMeshOrientation(
    Mesh Orientation Mode=::NewEnumerator2
    Rotation=&MakeVector(X=0, Y=0, Z=&RandomRangeFloat(Minimum=0, Maximum=1))
    Use Rotation Vector=true
   )
   &AddVelocity(Velocity=<0 0 200>)
  ]
  particleUpdateScript [
   &ParticleState()
   &DynamicMaterialParameters(
    Index 0 Param 1=&FloatFromCurve(FloatCurve=DataInterface() {propertyValues {Curve {keys [
         {arriveTangent=-1.5522, interpMode=RCIM_Cubic, leaveTangent=-1.5522, tangentMode=RCTM_User, time=0, value=1}
         {arriveTangent=-0.9602, interpMode=RCIM_Cubic, leaveTangent=-0.9602, tangentMode=RCTM_User, time=1, value=0}
        ]}}})
    Module.Param0WriteEnabled=true
    Module.Param1WriteEnabled=true
    Module.Param2WriteEnabled=true
    Module.Param3WriteEnabled=true
   )
   &UpdateMeshOrientation(Rotation Rate=3, Rotation Vector=<0 0 1>)
   &ScaleMeshSize(Scale Factor=&Multiply_VectorByFloat(
     Vector=&VectorFromCurve(
      VectorCurve=DataInterface() {propertyValues {XCurve {keys [<at 0 1.2>,<at 1 0.7>]}, YCurve {keys [<at 0 1.2>,<at 1 0.7>]}, ZCurve {keys [<at 0 0.5>,<at 1 1.5>]}}}
      Scale Curve=<0.9 0.9 1>
     )
     Float=1.5
    ))
   &ScaleColor(
    Scale Alpha=&FloatFromCurve(
     FloatCurve=DataInterface() {propertyValues {Curve {keys [
         {interpMode=RCIM_Cubic, tangentMode=RCTM_User, time=0, value=0}
         {interpMode=RCIM_Cubic, tangentMode=RCTM_User, time=0.5, value=1}
         {tangentMode=RCTM_User, time=1, value=0}
        ]}}}
     Scale Curve=4
    )
    Scale RGB=&Multiply_VectorByFloat(Vector=<1 1 1>, Float=&FNTimeOfDayScalar(Night Scale=0.7, Morning Scale=0.3, Evening Scale=0.3))
    ScaleA=true
    ScaleRGB=true
   )
   /SetVariables(Particles.Scale_RFlow_Color:LinearColor=&MakeLinearColorFromVectorAndFloat(
     "Vector (RGB)"=&MakeVectorFromLinearColorRGB(Color=<rgba 5 3.7468 1.5 1>)
     "Float (Alpha)"=&Multiply_Float(
      A=&Multiply_Float(A=0.8, B=&FNTimeOfDayScalar(Night Scale=0.7, Morning Scale=0.3, Evening Scale=0.3))
      B=&FloatFromCurve(FloatCurve=DataInterface() {propertyValues {Curve {keys [
           {interpMode=RCIM_Cubic, tangentMode=RCTM_User, time=0, value=0}
           {interpMode=RCIM_Cubic, tangentMode=RCTM_User, time=0.5, value=1}
           {tangentMode=RCTM_User, time=1, value=0}
          ]}}})
     )
    )) {isSetParameters=true}
   /SetVariables(Particles.Dyn_RFlow_Param:Vector4f=&MakeVector4(
     X=&FloatFromCurve(FloatCurve=DataInterface() {propertyValues {Curve {keys [
          {arriveTangent=-1.5522, interpMode=RCIM_Cubic, leaveTangent=-1.5522, tangentMode=RCTM_User, time=0, value=1}
          {arriveTangent=-1.5926, interpMode=RCIM_Cubic, leaveTangent=-1.5926, tangentMode=RCTM_User, time=1, value=0}
         ]}}})
     Y=0
     Z=0
     W=0
    )) {isSetParameters=true}
   &SolveForcesAndVelocity(Clamp Velocity=false, Limit Acceleration=false)
  ]
  rendererClasses [&NiagaraMeshRendererProperties]
  renderers [
   {rendererClass=&NiagaraMeshRendererProperties, rendererIndex=0}
   {rendererClass=&NiagaraMeshRendererProperties, rendererIndex=1}
  ]
 }
 RainbowFlare {
  emitterUpdateScript [
   &EmitterScalability(
    Module.bApplySpawnCountScale=false
    Module.bApplySpawnCountScaleByDistanceFraction=false
    Module.bUseMaxDistance=false
    Module.bUseMinDistance=false
   ) {enabled=false}
   &EmitterLifeCycle(MaxLoopCount=1, NextLoopDuration=1.7375) {enabled=false}
   &EmitterState()
   &SpawnRate(SpawnRate=10, Use Spawn Probability=false)
   &SpawnBurst_Instantaneous(Spawn Count=1, Use Spawn Probability=false) {enabled=false}
  ]
  particleSpawnScript [
   /Niagara/Modules/Spawn/Initialization/InitializeParticle(
    Color=<rgba 1 1 1 0.5>
    Lifetime=0.5
    Module.Write Color=true
    Module.Write Lifetime=true
    Module.Write Mass=true
    Module.Write Material Random=true
    Module.Write Position=true
    Module.Write Scale=false
    Module.Write SpriteRotation=true
    Module.Write SpriteSize=true
    Position=&SimulationPosition()
    Sprite Rotation=&UniformRangedFloat(Minimum=-360, Maximum=360)
    Sprite Size=&Vector2DFromFloat(Value=500)
   )
   &SystemLocation(MaskSpawnGroup=false, Offset=<0 0 100>)
   &CameraOffset(Camera Offset Amount=50)
  ]
  particleUpdateScript [
   &ParticleState()
   &SolveForcesAndVelocity(Clamp Velocity=false, Limit Acceleration=false)
   &ScaleColor(
    Scale Alpha=&FloatFromCurve(
     FloatCurve=DataInterface() {propertyValues {Curve {keys [
         {arriveTangent=6.4019, interpMode=RCIM_Cubic, leaveTangent=6.4019, tangentMode=RCTM_User, time=0, value=0}
         {arriveTangent=0.002, interpMode=RCIM_Cubic, leaveTangent=0.002, time=0.2, value=0.7}
         {arriveTangent=-0.9006, interpMode=RCIM_Cubic, leaveTangent=-0.9006, tangentMode=RCTM_User, time=1.0011, value=0.002}
        ]}}}
     Scale Curve=&FloatFromCurve(
      FloatCurve=DataInterface() {propertyValues {Curve {keys [
          {interpMode=RCIM_Cubic, tangentMode=RCTM_User, time=0, value=0}
          {interpMode=RCIM_Cubic, tangentMode=RCTM_User, time=0.5, value=1}
          {tangentMode=RCTM_User, time=1, value=0}
         ]}}}
      CurveIndex=$System.NormalizedLoopAge
      Scale Curve=2
     )
    )
    ScaleA=true
    ScaleRGB=false
   )
   &ScaleSpriteSize(Scale Factor=&Vector2DFromCurve(
     Vector2Curve=DataInterface() {propertyValues {
       XCurve {keys [
         {arriveTangent=0.0114, interpMode=RCIM_Cubic, leaveTangent=0.0114, tangentMode=RCTM_User, time=-0.0107, value=0.2}
         {arriveTangent=1.1193, interpMode=RCIM_Cubic, leaveTangent=1.1193, tangentMode=RCTM_User, time=1.0179, value=0.7}
        ]}
       YCurve {keys [
         {arriveTangent=0.0114, interpMode=RCIM_Cubic, leaveTangent=0.0114, tangentMode=RCTM_User, time=-0.0269, value=0.2}
         {arriveTangent=1.1193, interpMode=RCIM_Cubic, leaveTangent=1.1193, tangentMode=RCTM_User, time=1.0179, value=0.7}
        ]}
      }}
     ScaleCurve=<1 1>
    ))
   &DynamicMaterialParameters(
    Index 0 Param 1=10
    Index 0 Param 2=1
    Index 0 Param 3=1
    Index 0 Param 4=1
    Index 1 Param 2=0.2
    Index 1 Param 3=2
    Index 1 Param 4=&FloatFromCurve(FloatCurve=DataInterface() {propertyValues {Curve {keys [
         {arriveTangent=4.1785, interpMode=RCIM_Cubic, leaveTangent=4.1785, tangentMode=RCTM_User, time=0, value=0}
         {arriveTangent=1.1179, interpMode=RCIM_Cubic, leaveTangent=1.1179, tangentMode=RCTM_User, time=0.2042, value=0.6228}
         {arriveTangent=0.0224, interpMode=RCIM_Cubic, leaveTangent=0.0224, tangentMode=RCTM_User, time=1, value=1}
        ]}}})
    Module.Param0WriteEnabled=true
    Module.Param0WriteEnabled1=true
    Module.Param1WriteEnabled=true
    Module.Param1WriteEnabled1=true
    Module.Param2WriteEnabled=true
    Module.Param2WriteEnabled1=true
    Module.Param3WriteEnabled=true
    Module.Param3WriteEnabled1=true
    Write Parameter Index 1=true
   )
  ]
  rendererClasses [&NiagaraSpriteRendererProperties]
  renderers [{rendererClass=&NiagaraSpriteRendererProperties}]
 }
 RainbowFlare001 {
  emitterUpdateScript [
   &EmitterScalability(
    Module.bApplySpawnCountScale=false
    Module.bApplySpawnCountScaleByDistanceFraction=false
    Module.bUseMaxDistance=false
    Module.bUseMinDistance=false
   ) {enabled=false}
   &EmitterLifeCycle(MaxLoopCount=1, NextLoopDuration=1.7375) {enabled=false}
   &EmitterState()
   &SpawnRate(Use Spawn Probability=false) {enabled=false}
   &SpawnBurst_Instantaneous(Spawn Count=1, Use Spawn Probability=false)
  ]
  particleSpawnScript [
   &InitializeParticle(
    Lifetime=3
    Module.Write Color=true
    Module.Write Lifetime=true
    Module.Write Mass=true
    Module.Write Material Random=true
    Module.Write Position=true
    Module.Write Scale=false
    Module.Write SpriteRotation=false
    Module.Write SpriteSize=true
    Position=&SimulationPosition()
    Sprite Rotation=&UniformRangedFloat(Minimum=0, Maximum=0.5)
    Sprite Size=&Vector2DFromFloat(Value=250)
   )
   &SystemLocation(MaskSpawnGroup=false, Offset=<0 0 100>)
   &CameraOffset(Camera Offset Amount=50)
  ]
  particleUpdateScript [
   &ParticleState()
   &SolveForcesAndVelocity(Clamp Velocity=false, Limit Acceleration=false)
   &ScaleColor(
    Scale Alpha=&FloatFromCurve(
     FloatCurve=DataInterface() {propertyValues {Curve {keys [
         {arriveTangent=6.4019, interpMode=RCIM_Cubic, leaveTangent=6.4019, tangentMode=RCTM_User, time=0, value=0}
         {arriveTangent=0.002, interpMode=RCIM_Cubic, leaveTangent=0.002, time=0.2, value=0.7}
         {arriveTangent=-0.9006, interpMode=RCIM_Cubic, leaveTangent=-0.9006, tangentMode=RCTM_User, time=1.0011, value=0.002}
        ]}}}
     Scale Curve=&FloatFromCurve(
      FloatCurve=DataInterface() {propertyValues {Curve {keys [
          {arriveTangent=25.4717, interpMode=RCIM_Cubic, leaveTangent=25.4717, tangentMode=RCTM_User, time=0, value=0}
          {arriveTangent=0.1236, interpMode=RCIM_Cubic, leaveTangent=0.1236, tangentMode=RCTM_User, time=0.1, value=1}
          <at 0.5 1 RCIM_Cubic>
          {arriveTangent=-0.4595, interpMode=RCIM_Cubic, leaveTangent=-0.4595, tangentMode=RCTM_User, time=0.8, value=1}
          {arriveTangent=-26.3419, interpMode=RCIM_Cubic, leaveTangent=-26.3419, tangentMode=RCTM_User, time=0.9, value=0}
         ]}}}
      CurveIndex=$System.NormalizedLoopAge
      Scale Curve=1.5
     )
    )
    ScaleA=true
    ScaleRGB=false
   )
   &ScaleSpriteSize(Scale Factor=&Vector2DFromCurve(
     Vector2Curve=DataInterface() {propertyValues {
       XCurve {keys [{arriveTangent=0.3602, interpMode=RCIM_Cubic, leaveTangent=0.3602, tangentMode=RCTM_User, time=-0.0107, value=0.6}, <at 1.0179 0.7>]}
       YCurve {keys [{arriveTangent=0.3602, interpMode=RCIM_Cubic, leaveTangent=0.3602, tangentMode=RCTM_User, time=-0.0269, value=0.6}, <at 1.0179 0.7>]}
      }}
     ScaleCurve=<1 1>
    ))
   &SpriteRotationRate(Rotation Rate=&UniformRangedFloat(Minimum=-50, Maximum=50))
  ]
  rendererClasses [&NiagaraSpriteRendererProperties]
  renderers [{rendererClass=&NiagaraSpriteRendererProperties}]
 }
 Shockwave001 {
  emitterUpdateScript [
   &EmitterScalability(
    Module.bApplySpawnCountScale=false
    Module.bApplySpawnCountScaleByDistanceFraction=false
    Module.bUseMaxDistance=false
    Module.bUseMinDistance=false
   ) {enabled=false}
   &EmitterLifeCycle(NextLoopDuration=4.4208) {enabled=false}
   &EmitterState()
   &SpawnBurst_Instantaneous(Spawn Count=1, Spawn Time=0.133, Use Spawn Probability=false)
  ]
  particleSpawnScript [
   &InitializeParticle(
    Color=<rgba 15000 12214.4971 7500 0.25>
    Lifetime=0.3
    Mesh Scale=<0.35 0.35 0.35>
    Module.Write Color=true
    Module.Write Lifetime=true
    Module.Write Mass=true
    Module.Write Material Random=true
    Module.Write Position=false
    Module.Write Scale=true
    Module.Write SpriteRotation=false
    Module.Write SpriteSize=false
    Position=&SimulationPosition()
   )
   &SystemLocation(MaskSpawnGroup=false, Offset=<0 0 100>)
  ]
  particleUpdateScript [
   &ParticleState()
   &Color(Color=<rgba 5000 4291.6665 2500 1>)
   &DynamicMaterialParameters(
    Index 0 Param 1=&FloatFromCurve(FloatCurve=DataInterface() {propertyValues {Curve {keys [
         {arriveTangent=11.0061, interpMode=RCIM_Cubic, leaveTangent=11.0061, tangentMode=RCTM_User, time=-0.0005, value=-0.0141}
         {arriveTangent=1.9873, interpMode=RCIM_Cubic, leaveTangent=1.9873, tangentMode=RCTM_User, time=0.1692, value=1.0698}
         <at 1.0021 1.9784>
         <at 1.61 1.9784>
        ]}}})
    Index 0 Param 2=&FloatFromCurve(FloatCurve=DataInterface() {propertyValues {Curve {keys [<at -0.0005 15>,<at 0.1612 13.3874>,<at 1.0021 5>,<at 1.61 5>]}}})
    Index 0 Param 3=-0.8
    Module.Param0WriteEnabled=true
    Module.Param1WriteEnabled=true
    Module.Param2WriteEnabled=true
    Module.Param3WriteEnabled=true
   )
   &ScaleMeshSize(Scale Factor=&VectorFromCurve(VectorCurve=DataInterface() {propertyValues {
       XCurve {keys [
         {arriveTangent=9.7102, interpMode=RCIM_Cubic, leaveTangent=9.7102, tangentMode=RCTM_User, time=-0.0005, value=3.4203}
         {arriveTangent=6.8631, interpMode=RCIM_Cubic, leaveTangent=6.8631, tangentMode=RCTM_User, time=0.2646, value=6.6389}
         <at 0.9996 7.8749>
         <at 1.61 7.8749>
        ]}
       YCurve {keys [
         {arriveTangent=25.6031, interpMode=RCIM_Cubic, leaveTangent=25.6031, tangentMode=RCTM_User, time=-0.0005, value=3.4203}
         {arriveTangent=6.8631, interpMode=RCIM_Cubic, leaveTangent=6.8631, tangentMode=RCTM_User, time=0.2633, value=6.6865}
         <at 0.9996 7.8749>
         <at 1.61 7.8749>
        ]}
       ZCurve {keys [
         {arriveTangent=25.8257, interpMode=RCIM_Cubic, leaveTangent=25.8257, tangentMode=RCTM_User, time=-0.0005, value=3.4203}
         {arriveTangent=6.8631, interpMode=RCIM_Cubic, leaveTangent=6.8631, tangentMode=RCTM_User, time=0.2666, value=6.7585}
         <at 0.9996 7.8749>
         <at 1.61 7.8749>
        ]}
      }}))
   &ScaleColor(
    Scale Alpha=&FloatFromCurve(FloatCurve=DataInterface() {propertyValues {Curve {keys [
         {interpMode=RCIM_Cubic, tangentMode=RCTM_User, time=-0.0034, value=0.1203}
         {arriveTangent=-0.1606, interpMode=RCIM_Cubic, leaveTangent=-0.1606, tangentMode=RCTM_User, time=0.3345, value=0.0495}
         {interpMode=RCIM_Cubic, tangentMode=RCTM_User, time=0.9972, value=-0.0001}
        ]}}})
    Scale RGB=&ScaleFloatByCurve(
     InitialValue:Vector3f=<1 1 1>
     ScaleCurve:NiagaraDataInterfaceVectorCurve=DataInterface() {propertyValues {
       XCurve {keys [<at 0.0013 1.998>,<at 1 1>]}
       YCurve {keys [<at 0.0012 1.4636>,<at 1.0017 0.264>]}
       ZCurve {keys [<at 0.0012 1.0048>,<at 1.0017 0.264>]}
      }}
     CurveIndex:NiagaraFloat=$Particles.NormalizedAge
    )
    ScaleA=true
    ScaleRGB=true
   )
  ]
  rendererClasses [&NiagaraMeshRendererProperties]
  renderers [{rendererClass=&NiagaraMeshRendererProperties, rendererIndex=0}]
 }
 spikes {
  emitterUpdateScript [
   &EmitterState(Life Cycle Mode=::NewEnumerator1, Loop Behavior=::NewEnumerator1, Loop Delay=0.03, Loop Duration=1, UseLoopDelay=true)
   &SpawnBurst_Instantaneous(Spawn Count=10, Use Spawn Probability=false, UseLoopCountLimit=false)
  ]
  particleSpawnScript [
   /Niagara/Modules/Spawn/Initialization/V2/InitializeParticle(
    Color=<rgba 1 0.6933 0.2 1>
    Lifetime Max=0.25
    Lifetime Min=0.15
    Lifetime Mode=::NewEnumerator1
    Sprite Size Max=<100 180>
    Sprite Size Min=<20 100>
    Sprite Size Mode=::NewEnumerator2
    UsePositionOffset=true
    Write Material Random=true
   )
   &TorusLocation(Align Torus With Axis=false, Handle Radius=0, Large Radius=10, MaskSpawnGroup=false, Module.Override Randomness=false)
   &AddVelocityFromPoint(
    Module.Use Falloff Distance=false
    Origin Offset=&RandomRangeVector(Minimum=<0 0 -5>, Maximum=<0 0 -10>)
    Velocity Strength=&RandomRangeFloat(Minimum=600, Maximum=800)
   )
   &SubUVAnimation(End Frame=0, SubUV Animation Mode:ENiagara_SubUVLookupMode=::NewEnumerator2)
   &CameraOffset(Camera Offset Amount=10)
  ]
  particleUpdateScript [
   &ParticleState()
   &Drag(Drag=6, Use Linear Drag=true, Use Rotational Drag=false)
   &SolveForcesAndVelocity(Clamp Velocity=false, Limit Acceleration=false)
   &ScaleSpriteSize(
    Non-Uniform Curve Scale=&Multiply_Vector2D_ByFloat(Float=0.6, Vector2D=<1 1>)
    Non-Uniform Curve Sprite Scale=DataInterface() {propertyValues {
      XCurve {keys [{tangentMode=RCTM_User, time=0, value=1}, {tangentMode=RCTM_User, time=1, value=0.8}]}
      YCurve {keys [{tangentMode=RCTM_User, time=0, value=0.5}, {tangentMode=RCTM_User, time=1, value=1.7}]}
     }}
    Scale Sprite Size Mode=::NewEnumerator3
   )
   &ScaleColor(
    Scale Alpha=&FloatFromCurve(
     FloatCurve=DataInterface() {propertyValues {Curve {keys [<at 0 0>,<at 0.1 1>,<at 0.2 0.8>,<at 1 0>]}}}
     Scale Curve=&FNTimeOfDayScalar(Night Scale=1, Morning Scale=0.3, Evening Scale=0.5)
    )
    ScaleA=true
    ScaleRGB=true
   )
  ]
  rendererClasses [&NiagaraSpriteRendererProperties]
  renderers [{rendererClass=&NiagaraSpriteRendererProperties}]
 }
}
summary {
 systemName=NS_Emote_EmeraldGlassTransform_Might_COMBINED
 emitters [
  {emitterName=InkPop, rendererClasses [&NiagaraSpriteRendererProperties]}
  {emitterName=Explode_Sparkles, simTarget=GPUComputeSim, rendererClasses [&NiagaraSpriteRendererProperties]}
  {emitterName=HA, rendererClasses [&NiagaraSpriteRendererProperties]}
  {bEnabled=false, emitterName=HQ_Light, rendererClasses [&NiagaraLightRendererProperties]}
  {emitterName=Shockwave001, rendererClasses [&NiagaraMeshRendererProperties]}
  {emitterName=RainbowFlare001, rendererClasses [&NiagaraSpriteRendererProperties]}
  {emitterName=Flare_01001, rendererClasses [&NiagaraSpriteRendererProperties]}
  {emitterName=Flare_03001, rendererClasses [&NiagaraSpriteRendererProperties]}
  {emitterName=Decals, rendererClasses [&NiagaraSpriteRendererProperties]}
  {emitterName=spikes, rendererClasses [&NiagaraSpriteRendererProperties]}
  {emitterName=Explode_Sparkles001, simTarget=GPUComputeSim, rendererClasses [&NiagaraSpriteRendererProperties]}
  {emitterName=RainbowFlare, rendererClasses [&NiagaraSpriteRendererProperties]}
  {emitterName=OutwardSparkles, simTarget=GPUComputeSim, rendererClasses [&NiagaraSpriteRendererProperties]}
  {emitterName=BoxSparkle, simTarget=GPUComputeSim, rendererClasses [&NiagaraSpriteRendererProperties]}
  {emitterName=AURA_Combined, rendererClasses [&NiagaraMeshRendererProperties]}
  {emitterName=RFlow_Combined, rendererClasses [&NiagaraMeshRendererProperties]}
  {emitterName=Burst_Combined, rendererClasses [&NiagaraMeshRendererProperties]}
  {emitterName=Flare, rendererClasses [&NiagaraSpriteRendererProperties]}
  {emitterName=InkPop001, rendererClasses [&NiagaraSpriteRendererProperties]}
 ]
 userVariables [
  User.bIsFrontEnd {type=NiagaraBool}
  User.bIsFrontEndPreview {type=NiagaraBool}
  User.ForwardDot {type=NiagaraFloat, defaultValue=0, description=""}
  User.FrontendSpawn {type=NiagaraBool}
  User.LightningOrder {
   type=NiagaraDataInterfaceArrayInt32
   description=""
   defaultValue {
    struct=NiagaraExt_VariableValue_DataInterface
    value {
     dataInterface=&self:NiagaraDataInterfaceArrayInt32_0
     dataInterfaceClass=&NiagaraDataInterfaceArrayInt32
    }
   }
  }
  User.RibbonIDOffset {type=NiagaraInt32}
  User.SecondSpawnTiming {type=NiagaraFloat, defaultValue=0, description=""}
  User.SkeletalMesh {
   type=NiagaraDataInterfaceSkeletalMesh
   defaultValue {
    struct=NiagaraExt_VariableValue_DataInterface
    value {dataInterface=&self:NiagaraDataInterfaceSkeletalMesh_1}
   }
  }
 ]
}
