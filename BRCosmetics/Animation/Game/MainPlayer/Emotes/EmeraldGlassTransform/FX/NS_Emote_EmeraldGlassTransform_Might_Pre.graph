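# Niagara system for the EmeraldGlassTransform emote, apparently the "Might" pre-transform stage, producing a warm gold glow with pulsing energy.
# Rate-spawned, short-lived shockwave and ground-ring meshes expand and fade, alongside a burst-spawned glow sprite and a distance-culled light.
# Curves drive fades, scale and dynamic material parameters, while time-of-day and front-end/gameplay modifiers tune brightness and color.
# Skeletal mesh sampling modules are present but disabled.
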NiagaraSystem /BRCosmetics/Animation/Game/MainPlayer/Emotes/EmeraldGlassTransform/FX/NS_Emote_EmeraldGlassTransform_Might_Pre
dependencies {
 usedDataInterfaces [&NiagaraDataInterfaceCurve, &NiagaraDataInterfaceVectorCurve, &NiagaraDataInterfaceColorCurve]
 usedDynamicInputs []
 usedModules [
  &SystemState
  &self:SystemSpawnScript.SystemScriptSource.SystemScriptGraph.NiagaraNodeAssignment_0.SetVariables_#1
 ]
}
emitters {
 Decals {
  emitterUpdateScript [
   &EmitterState()
   &SpawnBurst_Instantaneous(Spawn Count=1, Use Spawn Probability=false) {module=SpawnBurst_Instantaneous001}
  ]
  particleSpawnScript [
   &FN_SetSpawnGroup(Number of Spawn Groups=23) {module=FN_SetSpawnGroup001}
   &InitializeParticle(
    Color=&FNTimeOfDayColor(Night Color=<rgba 5 4.1702 1.5 1>, Morning Color=<rgba 10 7.629 0 1>, Day Color=<rgba 10 7.629 0 1>, Evening Color=<rgba 5 4.1702 1.5 1>)
    Position Offset=<0 0 117>
    Position Offset Coordinate Space=::Local
    Sprite Size=<200 220>
    Sprite Size Mode=::NewEnumerator4
    UsePositionOffset=true
    Write Material Random=true
   ) {module=InitializeParticle001}
  ]
  particleUpdateScript [
   &SkeletalMeshLocation(
    Bone / Socket Index=$Particles.SpawnGroup
    Bone Sampling Mode=::NewEnumerator4
    MaskSpawnGroup=false
    Sampled Position Offset=<0 0 10>
    Sampled Position Offset Coordinate Space:ENiagaraSkelMeshTransforms=::NewEnumerator2
    Skeletal Mesh=DataInterface() {propertyValues {
      DefaultMesh=null
      ExcludeBoneName=root
      OptionalComponentClass=null
      PreviewMesh=/BRCosmetics/Characters/Player/Female/Medium/Bodies/F_MED_Bistro_Monster/Meshes/F_MED_Bistro_Monster
      SkinningMode=SkinOnTheFly
      SoftSourceActor=null
      SourceMode=Default
      UvSetIndex=0
      WholeMeshLOD=-1
      bExcludeBone=false
      bReadDeformedGeometry=true
      bRequireCurrentFrameData=true
      ComponentTags []
      FilteredBones [
       head
       spine_05
       spine_03
       spine_01
       clavicle_l
       clavicle_r
       upperarm_l
       upperarm_r
       upperarm_twist_01_l
       upperarm_twist_01_r
       lowerarm_l
       lowerarm_r
       lowerarm_twist_01_l
       lowerarm_twist_01_r
       hand_l
       hand_r
       pelvis
       thigh_l
       thigh_r
       thigh_twist_01_l
       thigh_twist_01_r
       calf_l
       calf_r
      ]
      FilteredSockets []
      MeshUserParameter {parameter {name=None, typeDefHandle {registeredTypeIndex=1}, varData []}}
      SamplingRegions []
     }}
   ) {enabled=false}
   &ParticleState()
   &CameraOffset(Camera Offset Amount=55)
   &DynamicMaterialParameters(
    Index 0 Param 1=2
    Index 0 Param 2=1
    Index 0 Param 3=&FloatFromCurve(FloatCurve=DataInterface() {propertyValues {Curve {keys [<at 0 -0.2>,<at 0.7 0.3>]}}})
    Module.Param0WriteEnabled=true
    Module.Param1WriteEnabled=true
    Module.Param2WriteEnabled=true
    Module.Param3WriteEnabled=true
   )
   &ScaleColor(
    Scale Alpha=&FloatFromCurve(
     FloatCurve=DataInterface() {propertyValues {Curve {keys [
         <at 0 0 RCIM_Cubic>
         {arriveTangent=1.9398, interpMode=RCIM_Cubic, leaveTangent=1.9398, tangentMode=RCTM_User, time=0.4835, value=0.3}
         <at 0.5835 1>
         {arriveTangent=-3.5366, interpMode=RCIM_Cubic, leaveTangent=-3.5366, tangentMode=RCTM_User, time=0.65, value=1}
         <at 1 0>
        ]}}}
     Scale Curve=&Multiply_Float(
      A=&FNTimeOfDayScalar(Night Scale=0.7, Morning Scale=0.3, Evening Scale=0.4)
      B=&FN_FrontEndFloatModifier(Gameplay=3, Front End=2)
     )
    )
    Scale RGB=&MakeVectorFromLinearColorRGB(Color=&ColorFromCurve(ColorCurve=DataInterface() {propertyValues {
        AlphaCurve {keys [<at 0 1 RCIM_Cubic>]}
        BlueCurve {keys [<at 0 1 RCIM_Cubic>, {arriveTangent=4.4002, interpMode=RCIM_Cubic, leaveTangent=4.4002, tangentMode=RCTM_User, time=1, value=3}]}
        GreenCurve {keys [<at 0 1 RCIM_Cubic>, {arriveTangent=4.4002, interpMode=RCIM_Cubic, leaveTangent=4.4002, tangentMode=RCTM_User, time=1, value=3}]}
        RedCurve {keys [<at 0 1 RCIM_Cubic>, {arriveTangent=4.4002, interpMode=RCIM_Cubic, leaveTangent=4.4002, tangentMode=RCTM_User, time=1, value=3}]}
       }}))
    ScaleA=true
    ScaleRGB=true
   )
   &ScaleColor(
    Scale RGB=&VectorFromFloat(Value=&FloatFromCurve(FloatCurve=DataInterface() {propertyValues {Curve {keys [
          <at 0 0>
          {arriveTangent=39.28, interpMode=RCIM_Cubic, leaveTangent=29.7857, tangentMode=RCTM_Break, time=0.4, value=3}
          {arriveTangent=3.0581, interpMode=RCIM_Cubic, leaveTangent=3.0581, tangentMode=RCTM_User, time=0.4, value=2.5}
          <at 0.8 7>
          {arriveTangent=-5.1937, interpMode=RCIM_Cubic, leaveTangent=-5.1937, tangentMode=RCTM_User, time=0.85, value=1}
          <at 1 0.5 RCIM_Cubic>
         ]}}}))
    ScaleA=true
    ScaleRGB=true
   ) {enabled=false, module=ScaleColor001}
  ]
  rendererClasses [&NiagaraSpriteRendererProperties]
  renderers [{rendererClass=&NiagaraSpriteRendererProperties}]
 }
 HQ_Light {
  emitterUpdateScript [
   &EmitterState(Enable Distance Culling=true, MaxDistance=1500, Module.bUseMaxDistance=true, Module.bUseMinDistance=false, Scalability Mode=::NewEnumerator1)
   &SpawnBurst_Instantaneous(Spawn Count=1, Spawn Time=0.133, Use Spawn Probability=false)
  ]
  particleSpawnScript [&InitializeParticle(Color Mode=::NewEnumerator0, Lifetime=$System.CurrentLoopDuration, UsePositionOffset=false, Write Material Random=true)]
  particleUpdateScript [
   &ParticleState()
   &FN_Light_Attributes(
    Color=&MakeLinearColorFromVectorAndFloat(
     "Vector (RGB)"=&FN_FrontEndVector3Modifier(Gameplay=<2 1.6 1.4>, Front End=<0.5 0.1 0>)
     "Float (Alpha)"=1
    )
    Create Color Attribute=true
    Create Position Attribute=true
    Create Radius Attribute=true
    Enable Visibility Tag=true
    Position=&Add_Vector(
     A=Linked() {linkedVariable {name=Particles.Position, type=NiagaraPosition}}
     B=&TransformVector(Vector=<0 0 50>, Source Space=::World, Destination Space=::Local)
    )
    Radius=250
    Scale Color=&FloatFromCurve(
     FloatCurve=DataInterface() {propertyValues {Curve {keys [
         {arriveTangent=117.3003, interpMode=RCIM_Cubic, leaveTangent=117.3003, tangentMode=RCTM_User, time=0, value=0}
         {arriveTangent=1.3594, interpMode=RCIM_Cubic, leaveTangent=1.3594, tangentMode=RCTM_User, time=0.0286, value=1}
         {arriveTangent=-1.0294, interpMode=RCIM_Cubic, leaveTangent=-1.0294, time=0.95, value=0.9805}
         {arriveTangent=-65.8727, interpMode=RCIM_Cubic, leaveTangent=-65.8727, tangentMode=RCTM_User, time=1, value=0}
        ]}}}
     Scale Curve=&Multiply_Float(
      A=&FNTimeOfDayScalar(Night Scale=0.7, Morning Scale=0.3, Evening Scale=0.4)
      B=&FN_FrontEndFloatModifier(Gameplay=4, Front End=3)
     )
    )
    VisibilityTag=HlslExpression() {hlslExpression="Particles.UniqueID && 1"}
   )
  ]
  rendererClasses [&NiagaraLightRendererProperties]
  renderers [{rendererClass=&NiagaraLightRendererProperties, rendererIndex=0}]
 }
 Shock_wave001 {
  emitterUpdateScript [
   &EmitterState()
   &SpawnRate(SpawnRate=7, Use Spawn Probability=false)
  ]
  particleSpawnScript [
   /Niagara/Modules/Spawn/Initialization/InitializeParticle(
    Color=<rgba 1 0.8445 0.1146 4>
    Lifetime=0.2
    Mesh Scale=&RandomRangeVector(Minimum=<2.2 2.7 2.7>, Maximum=<2.4 2.5 2.5>)
    Module.Write Color=true
    Module.Write Lifetime=true
    Module.Write Mass=true
    Module.Write Material Random=true
    Module.Write Position=true
    Module.Write Scale=true
    Module.Write SpriteRotation=false
    Module.Write SpriteSize=true
    Position=&Add_Vector(A=&SimulationPosition(), B=<0 0 -2>)
    Sprite Size=<5 5>
   )
   &SystemLocation(MaskSpawnGroup=false, Offset=<0 0 20>) {enabled=false}
   &InitialMeshOrientation(Rotation=<0 0.25 0>, Use Orientation Vector=true, Use Rotation Vector=true)
   &InitialMeshOrientation(
    Rotation=&RandomRangeVector(Minimum=<0 0 0>, Maximum=<0 0 0.15>)
    Use Orientation Vector=true
    Use Rotation Vector=true
   ) {module=InitialMeshOrientation001}
  ]
  particleUpdateScript [
   &ParticleState()
   &SkeletalMeshLocation(Bone Sampling Mode=::NewEnumerator4, MaskSpawnGroup=false, Orientation Sampling=::NewEnumerator1, Sampled Position Offset=<0 0 20>) {enabled=false}
   &SolveForcesAndVelocity(Clamp Velocity=false, Limit Acceleration=false)
   &ScaleColor(
    Scale Alpha=&Multiply_Float(
     A=1
     B=&Multiply_Float(
      A=&FloatFromCurve(FloatCurve=DataInterface() {propertyValues {Curve {keys [<at 0 0 RCIM_Cubic>,<at 0.026 1 RCIM_Cubic>,<at 1 0>]}}})
      B=&FNTimeOfDayScalar(Night Scale=0.5, Morning Scale=0.2, Evening Scale=0.2)
     )
    )
    ScaleA=true
    ScaleRGB=true
   ) {module=ScaleColor001}
   &ScaleMeshSize(Scale Factor=&VectorFromCurve(
     VectorCurve=DataInterface() {propertyValues {XCurve {keys [<at 0 0.5>,<at 1 1.5>]}, YCurve {keys [<at 0 0.7>,<at 1 1.6>]}, ZCurve {keys [<at 0 0.7>,<at 1 2>]}}}
     Scale Curve=&Multiply_Vector(A=<0.7 0.7 0.7>, B=$Engine.Owner.Scale)
    ))
   &DynamicMaterialParameters(
    Index 0 Param 1=&FloatFromCurve(FloatCurve=DataInterface() {propertyValues {Curve {keys [<at 0 0 RCIM_Cubic>,<at 1 0.28 RCIM_Cubic>]}}})
    Index 0 Param 2=&RandomRangeFloat(Minimum=0, Maximum=1)
    Index 0 Param 3=50
    Index 0 Param 4=&FloatFromCurve(FloatCurve=DataInterface() {propertyValues {Curve {keys [<at 0 0>, {arriveTangent=-3.0374, leaveTangent=2.0371, time=0.2, value=0}, {interpMode=RCIM_Cubic, tangentMode=RCTM_Break, time=1, value=0}]}}})
    Index 1 Param 1=0.5
    Module.Param0WriteEnabled=true
    Module.Param0WriteEnabled1=true
    Module.Param1WriteEnabled=true
    Module.Param1WriteEnabled1=true
    Module.Param2WriteEnabled=true
    Module.Param2WriteEnabled1=true
    Module.Param3WriteEnabled=true
    Module.Param3WriteEnabled1=true
    Write Parameter Index 1=true
   )
   /SetVariables(Particles.New Position:NiagaraPosition=$Particles.Position) {enabled=false, isSetParameters=true}
  ]
  rendererClasses [&NiagaraMeshRendererProperties]
  renderers [{rendererClass=&NiagaraMeshRendererProperties, rendererIndex=0}]
 }
 aura_ring_ground {
  emitterUpdateScript [
   &EmitterState()
   &SpawnRate(SpawnRate=5, Use Spawn Probability=false)
  ]
  particleSpawnScript [
   /Niagara/Modules/Spawn/Initialization/V2/InitializeParticle(
    Color=<rgba 2 1.6 1 1>
    Lifetime Max=0.5
    Lifetime Min=0.4
    Lifetime Mode=::NewEnumerator1
    Mesh Scale=&Multiply_VectorByFloat(
     Vector=&RandomRangeVector(Minimum=<1 1 1>, Maximum=<0.5 1 1>)
     Float=&RandomRangeFloat(Minimum=2, Maximum=2)
    )
    Mesh Scale Mode=::NewEnumerator4
    Position Offset=<0 0 -5>
    Position Offset Coordinate Space=::Local
    UsePositionOffset=true
    Write Material Random=true
   )
   &InitialMeshOrientation(Mesh Orientation Mode=::NewEnumerator2, Rotation=<0 0.25 0>, Use Rotation Vector=true)
   &BoxLocation(
    Box Offset=&FN_FrontEndVector3Modifier(Gameplay=<0 0 20>, Front End=<0 0 10>)
    Box Size=<0 0 0>
    MaskSpawnGroup=false
    Module.Override Seed=false
    Rotate Box=false
    Surface Only=false
   )
  ]
  particleUpdateScript [
   &ParticleState()
   &DynamicMaterialParameters(
    Index 0 Param 1=1
    Index 0 Param 2=3
    Index 0 Param 3=&FN_FrontEndFloatModifier(Gameplay=1, Front End=2)
    Index 0 Param 4=&FN_FrontEndFloatModifier(Gameplay=25, Front End=400)
    Module.Param0WriteEnabled=true
    Module.Param1WriteEnabled=true
    Module.Param2WriteEnabled=true
    Module.Param3WriteEnabled=true
   )
   &ScaleMeshSize(Scale Factor=&Multiply_VectorByFloat(
     Vector=<1 1 1>
     Float=&FloatFromCurve(
      FloatCurve=DataInterface() {propertyValues {Curve {keys [{interpMode=RCIM_Cubic, tangentMode=RCTM_User, time=0, value=1}, {interpMode=RCIM_Cubic, tangentMode=RCTM_User, time=1, value=1.5}]}}}
      Scale Curve=0.8
     )
    ))
   &ScaleColor(
    Scale Alpha=&FloatFromCurve(FloatCurve=DataInterface() {propertyValues {Curve {keys [
         {interpMode=RCIM_Cubic, tangentMode=RCTM_User, time=0, value=0}
         {interpMode=RCIM_Cubic, tangentMode=RCTM_User, time=0.3, value=0.75}
         {tangentMode=RCTM_User, time=1, value=0}
        ]}}})
    Scale RGB=&Multiply_VectorByFloat(
     Vector=<1 1 1>
     Float=&Multiply_Float(
      A=&FNTimeOfDayScalar(Night Scale=0.7, Morning Scale=0.3, Evening Scale=0.4)
      B=&FN_FrontEndFloatModifier(Gameplay=1, Front End=2)
     )
    )
    ScaleA=true
    ScaleRGB=true
   )
  ]
  rendererClasses [&NiagaraMeshRendererProperties]
  renderers [{rendererClass=&NiagaraMeshRendererProperties, rendererIndex=0}]
 }
}
summary {
 systemName=NS_Emote_EmeraldGlassTransform_Might_Pre
 emitters [
  {emitterName=aura_ring_ground, rendererClasses [&NiagaraMeshRendererProperties]}
  {emitterName=Shock_wave001, rendererClasses [&NiagaraMeshRendererProperties]}
  {emitterName=HQ_Light, rendererClasses [&NiagaraLightRendererProperties]}
  {emitterName=Decals, rendererClasses [&NiagaraSpriteRendererProperties]}
 ]
 userVariables [
  User.bIsFrontEnd {type=NiagaraBool}
  User.bIsFrontEndPreview {type=NiagaraBool}
  User.ForwardDot {type=NiagaraFloat, defaultValue=0, description=""}
  User.FrontendSpawn {type=NiagaraBool}
  User.RibbonIDOffset {type=NiagaraInt32}
  User.SecondSpawnTiming {type=NiagaraFloat, defaultValue=0, description=""}
  User.SkeletalMesh {
   type=NiagaraDataInterfaceSkeletalMesh
   defaultValue {
    struct=NiagaraExt_VariableValue_DataInterface
    value {dataInterface=&self:NiagaraDataInterfaceSkeletalMesh_1}
   }
  }
 ]
}
