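# Niagara burst effect for the SkinnyMight variant of the EmeraldGlassTransform emote, wrapping the character in toon smoke with a brief flash.
# Three outlined toon mesh smoke emitters sample User.SkeletalMesh bones for partitioned rising puffs, an expanding ring and scattered puffs, all shrinking and darkening over life.
# Spinning airflow meshes, warm spike sprites and a short distance-culled light add impact.
# Airflow brightens in front end; spike alpha scales with time of day.
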NiagaraSystem /BRCosmetics/Animation/Game/MainPlayer/Emotes/EmeraldGlassTransform/FX/NS_Emote_EmeraldGlassTransform_SkinnyMight
dependencies {
 usedDataInterfaces [
  &NiagaraDataInterfaceArrayFloat3
  &NiagaraDataInterfaceArrayFloat
  &NiagaraDataInterfaceCurve
  &NiagaraDataInterfaceColorCurve
  &NiagaraDataInterfaceVector2DCurve
 ]
 usedDynamicInputs []
 usedModules [&SystemState]
}
emitters {
 BigSmok {
  emitterSpawnScript [
   /Game/Effects/Niagara/MasterMaterials/ToonShader/Template/NM_ToonShaderParameters(HighLight_Softness=0, Lighting_Hardness=0.8, Metallic_LightingPower=0, RimLight_Width=0)
   /Game/Effects/Niagara/MasterMaterials/ToonShader/Template/NM_OutlineParameters()
  ]
  emitterUpdateScript [
   &EmitterState()
   &SpawnBurst_Instantaneous(Spawn Count=12, Use Spawn Probability=false, UseLoopCountLimit=false)
   &SpawnRate(
    Spawn Probability=0.75
    SpawnRate=&FloatFromCurve(FloatCurve=DataInterface() {propertyValues {Curve {keys [<at 0.2 0>,<at 0.2 1>]}}}, CurveIndex=$Emitter.NormalizedLoopAge, Scale Curve=40)
    Use Spawn Probability=true
   ) {enabled=false}
  ]
  particleSpawnScript [
   &FN_SetSpawnGroup(Number of Spawn Groups=2) {enabled=false}
   &PartitionParticles(Number Of Partitions=3)
   &InitializeParticle(
    Color=&MakeLinearColorFromVectorAndFloat("Vector (RGB)"=<1 1 1>, "Float (Alpha)"=1)
    Lifetime=&Multiply_Float(A=0.7, B=&RandomRangeFloat(Minimum=0.6, Maximum=0.8))
    Mesh Scale=&VectorFromFloat(Value=&SelectFloatFromArray(
      Array Sampling Mode=::NewEnumerator0
      Direct Array Index=?
      Float Selection Array=DataInterface() {propertyValues {FloatData [0.5,0.5,0.5]}}
      Random Seed=0
     ))
    Mesh Scale Mode=::NewEnumerator4
    Position Offset=&SelectVectorFromArray(
     Direct Array Index=&SpawnGroup()
     Vector Selection Array=DataInterface() {propertyValues {FloatData [<60 130 -70>, <-40 130 -70>]}}
    )
    Sprite Size Mode=::NewEnumerator3
    Uniform Sprite Size=&Multiply_Float(A=77, B=&MakeFloatFromVector(Vector=$Engine.Owner.Scale, Channel=::NewEnumerator0))
    UsePositionOffset=true
    Write Material Random=true
   )
   &InitialMeshOrientation(
    Rotation=&MakeVector(X=&RandomRangeFloat(Minimum=0, Maximum=1), Y=0, Z=0)
    Use Rotation Vector=true
   )
   &SkeletalMeshLocation(
    Bone / Socket Index=$Particles.PartitionParticles.Partition
    Bone Sampling Mode=::NewEnumerator4
    MaskSpawnGroup=false
    Sampled Position Offset=&SelectVectorFromArray(
     Direct Array Index=$Particles.PartitionParticles.Partition
     Vector Selection Array=DataInterface() {propertyValues {FloatData [<0 0 60>, <0 0 0>, <0 0 -90>]}}
    )
    Sampled Position Offset Coordinate Space:ENiagaraSkelMeshTransforms=::NewEnumerator2
    Skeletal Mesh=$User.SkeletalMesh
   )
   &ShapeLocation(
    Cylinder Height=1
    Cylinder Radius=&SelectFloatFromArray(Direct Array Index=$Particles.PartitionParticles.Partition, Float Selection Array=DataInterface() {propertyValues {FloatData [1,1,1]}})
    MaskSpawnGroup=false
    Override Seed=false
    Shape Primitive=::NewEnumerator1
    Surface Only=false
    UseOwnerScale=true
   ) {module=ShapeLocation001}
   &AddVelocity(
    Module.Use Falloff Distance=false
    Override Seed=false
    Velocity Mode=::NewEnumerator1
    Velocity Origin=$Particles.SkeletalMeshLocation.SampledPosition
    Velocity Speed=50
   ) {enabled=false}
  ]
  particleUpdateScript [
   &ParticleState()
   &ScaleMeshSize(Scale Factor=&VectorFromFloat(Value=&FloatFromCurve(
      FloatCurve=DataInterface() {propertyValues {Curve {keys [
          {tangentMode=RCTM_User, time=0, value=1}
          <at 0.1 2>
          {arriveTangent=-0.0405, interpMode=RCIM_Cubic, leaveTangent=-0.0405, tangentMode=RCTM_User, time=0.3, value=0.8}
          <at 0.8 0.5>
          {arriveTangent=-1.4118, interpMode=RCIM_Cubic, leaveTangent=-1.4118, tangentMode=RCTM_User, time=1, value=0}
         ]}}}
      Scale Curve=&Multiply_Float(
       A=0.5
       B=&SelectFloatFromArray(Direct Array Index=$Particles.PartitionParticles.Partition, Float Selection Array=DataInterface() {propertyValues {FloatData [5,7,8]}})
      )
     ))) {module=ScaleMeshSize002}
   &ScaleMeshSize(Scale Factor=&VectorFromFloat(Value=&FloatFromCurve(
      FloatCurve=DataInterface() {propertyValues {Curve {keys [
          <at 0 0.4 RCIM_Cubic>
          {arriveTangent=-0.0088, interpMode=RCIM_Cubic, leaveTangent=-0.0088, tangentMode=RCTM_User, time=0.7, value=1}
          {arriveTangent=-10.4257, interpMode=RCIM_Cubic, leaveTangent=-10.4257, tangentMode=RCTM_User, time=1, value=0}
         ]}}}
      Scale Curve=&MakeCustomFloatFromBool(Bool=&SetBoolByFloatComparison(A=1, B=0.4), True Float=1.25, False Float=1)
     ))) {enabled=false}
   &ScaleMeshSize(Scale Factor=&MakeVector(
     X=&FloatFromCurve(FloatCurve=DataInterface() {propertyValues {Curve {keys [{interpMode=RCIM_Cubic, tangentMode=RCTM_User, time=0, value=3}, {interpMode=RCIM_Cubic, tangentMode=RCTM_User, time=0.6, value=1}]}}})
     Y=1
     Z=1
    )) {enabled=false, module=ScaleMeshSize001}
   &Color(
    Color=&ColorFromCurve(ColorCurve=DataInterface() {propertyValues {
       AlphaCurve {keys [<at 0.1 1>]}
       BlueCurve {keys [<at 0 0.3643>, {arriveTangent=0.0577, time=0.333, value=0.2745}, <at 1 0.05>]}
       GreenCurve {keys [<at 0 0.3933>, {arriveTangent=-0.0961, time=0.333, value=0.3362}, <at 1 0.05>]}
       RedCurve {keys [<at 0 0.5781>, {arriveTangent=-1.0957, time=0.333, value=0.4583}, <at 1 0.05>]}
      }})
    Scale Alpha=&MakeFloatFromLinearColor(LinearColor=$Particles.Initial.Color, Channel:ENiagaraLinearColor_Channels=::NewEnumerator3)
   )
   &UpdateMeshOrientation(
    Rotation Rate=&FloatFromCurve(
     FloatCurve=DataInterface() {propertyValues {Curve {keys [{interpMode=RCIM_Cubic, tangentMode=RCTM_User, time=0.4, value=0}, {interpMode=RCIM_Cubic, tangentMode=RCTM_User, time=1, value=1}]}}}
     Scale Curve=&RandomRangeFloat(Minimum=-2, Maximum=2)
    )
    Rotation Vector=&RandomVector(Vector Scale=1)
   )
   &DynamicMaterialParameters(
    Index 0 Param 1=&FloatFromCurve(FloatCurve=DataInterface() {propertyValues {Curve {keys [
         {interpMode=RCIM_Cubic, tangentMode=RCTM_User, time=0, value=0}
         {arriveTangent=1.3695, interpMode=RCIM_Cubic, leaveTangent=1.3695, tangentMode=RCTM_User, time=1, value=1}
        ]}}})
    Module.Param0WriteEnabled=true
    Module.Param1WriteEnabled=false
    Module.Param2WriteEnabled=false
    Module.Param3WriteEnabled=false
   ) {enabled=false}
   /SetVariables(Particles.Position=&AddVectorToPosition(
     Position=$Particles.Position
     Vector=&Multiply_VectorByFloat(
      Vector=&NormalizeVector(
       Vector To Normalize=&Lerp_Vector(
        A=&Add_Vector(
         A=<0 1 0>
         B=&Multiply_VectorByFloat(Vector=&RandomVector(Vector Scale=1), Float=0.1)
        )
        B=&TransformVector(
         Vector=&NormalizeVector(
          Vector To Normalize=&Multiply_Vector(A=$Engine.Owner.Velocity, B=<-1 -1 -1>)
          Safe Normalize=true
          Safe Fallback Threshold=0
          Safe Fallback Vector=<0 0 0.001>
         )
         Source Space=::World
         Destination Space=::Local
        )
        Alpha=&FloatFromCurve(FloatCurve=DataInterface() {propertyValues {Curve {keys [
             {interpMode=RCIM_Cubic, leaveTangent=-0.0203, tangentMode=RCTM_Break, time=0, value=0}
             {arriveTangent=-0.0124, interpMode=RCIM_Cubic, leaveTangent=-0.0124, tangentMode=RCTM_User, time=1, value=1}
            ]}}})
       )
       Safe Normalize=true
       Safe Fallback Threshold=0
       Safe Fallback Vector=<0 0 0.001>
      )
      Float=&FloatFromCurve(
       FloatCurve=DataInterface() {propertyValues {Curve {keys [
           {interpMode=RCIM_Cubic, tangentMode=RCTM_User, time=0, value=1.425}
           {arriveTangent=-0.0838, interpMode=RCIM_Cubic, leaveTangent=-0.0838, tangentMode=RCTM_User, time=1, value=0.3}
          ]}}}
       Scale Curve=&Multiply_Float(
        A=5
        B=&MakeCustomFloatFromBool(Bool=&SetBoolByFloatComparison(A=$User.ForwardDot, B=0.4), True Float=1.5, False Float=1)
       )
      )
     )
    )) {enabled=false, isSetParameters=true, module=SetVariables_#1}
   &CalculateAccurateVelocity() {enabled=false}
   &GravityForce(Gravity=<0 0 300>)
   &Drag(Drag=8, Use Linear Drag=true, Use Rotational Drag=true)
   &SolveForcesAndVelocity(Clamp Velocity=false, Limit Acceleration=false)
  ]
  rendererClasses [&NiagaraMeshRendererProperties]
  renderers [
   {rendererClass=&NiagaraMeshRendererProperties, rendererIndex=0}
   {rendererClass=&NiagaraMeshRendererProperties, rendererIndex=1}
  ]
 }
 HQ_Light {
  emitterUpdateScript [
   &EmitterState(Enable Distance Culling=true, MaxDistance=1500, Module.bUseMaxDistance=true, Module.bUseMinDistance=false, Scalability Mode=::NewEnumerator1)
   &SpawnBurst_Instantaneous(Spawn Count=1, Use Spawn Probability=false)
  ]
  particleSpawnScript [&InitializeParticle(Color Mode=::NewEnumerator0, Lifetime=0.15, UsePositionOffset=false, Write Material Random=true)]
  particleUpdateScript [
   &ParticleState()
   &FN_Light_Attributes(
    Color=<rgba 2 1.3831 0.3791 1>
    Create Color Attribute=true
    Create Position Attribute=true
    Create Radius Attribute=true
    Enable Visibility Tag=true
    Position=&Add_Vector(
     A=Linked() {linkedVariable {name=Particles.Position, type=NiagaraPosition}}
     B=&TransformVector(Vector=<0 0 50>, Source Space=::World, Destination Space=::Local)
    )
    Radius=200
    Scale Color=&FloatFromCurve(
     FloatCurve=DataInterface() {propertyValues {Curve {keys [{tangentMode=RCTM_User, time=0, value=1}, {tangentMode=RCTM_User, time=1, value=0}]}}}
     Scale Curve=10
    )
    VisibilityTag=HlslExpression() {hlslExpression="Particles.UniqueID && 1"}
   )
   &FN_Gen5_Light(
    Attenutation:NiagaraFloat=&FloatFromCurve(
     FloatCurve=DataInterface() {propertyValues {Curve {keys [
         {arriveTangent=-1.1596, interpMode=RCIM_Cubic, time=0, value=1}
         {arriveTangent=-0.6519, interpMode=RCIM_Cubic, leaveTangent=-0.6519, tangentMode=RCTM_User, time=1, value=0.7}
        ]}}}
     Scale Curve=800
    )
    Color:LinearColor=<rgba 10 6.895 0.3963 1>
    Intensity:NiagaraFloat=&FloatFromCurve(
     FloatCurve=DataInterface() {propertyValues {Curve {keys [
         {arriveTangent=117.3003, interpMode=RCIM_Cubic, leaveTangent=117.3003, tangentMode=RCTM_User, time=0, value=0}
         {arriveTangent=1.3594, interpMode=RCIM_Cubic, leaveTangent=1.3594, tangentMode=RCTM_User, time=0.0286, value=1}
         {arriveTangent=-1.0294, interpMode=RCIM_Cubic, leaveTangent=-1.0294, time=0.95, value=0.9805}
         {arriveTangent=-65.8727, interpMode=RCIM_Cubic, leaveTangent=-65.8727, tangentMode=RCTM_User, time=1, value=0}
        ]}}}
     Scale Curve=45.4244
    )
    Position Offset:Vector3f=&TransformVector(Vector=<0 0 50>, Source Space=::World, Destination Space=::Local)
   ) {enabled=false}
  ]
  rendererClasses [&NiagaraLightRendererProperties]
  renderers [{rendererClass=&NiagaraLightRendererProperties, rendererIndex=0}]
 }
 SmokeRing {
  emitterSpawnScript [
   &NM_ToonShaderParameters(HighLight_Softness=0, Lighting_Hardness=0.8, Metallic_LightingPower=0, RimLight_Width=0)
   &NM_OutlineParameters()
  ]
  emitterUpdateScript [
   &EmitterState()
   &SpawnBurst_Instantaneous(Spawn Count=20, Use Spawn Probability=false, UseLoopCountLimit=false)
   &SpawnRate(Spawn Probability=0.75, SpawnRate=400, Use Spawn Probability=true) {enabled=false}
  ]
  particleSpawnScript [
   &FN_SetSpawnGroup(Number of Spawn Groups=2)
   &InitializeParticle(
    Color=&MakeLinearColorFromVectorAndFloat("Vector (RGB)"=<1 1 1>, "Float (Alpha)"=1)
    Lifetime=&RandomRangeFloat(Minimum=0.6, Maximum=1.2)
    Mesh Scale=&VectorFromFloat(Value=&SelectFloatFromArray(
      Array Sampling Mode=::NewEnumerator0
      Direct Array Index=?
      Float Selection Array=DataInterface() {propertyValues {FloatData [0.3,0.6,1]}}
      Random Seed=0
     ))
    Mesh Scale Mode=::NewEnumerator4
    Position Offset=&SelectVectorFromArray(
     Direct Array Index=&SpawnGroup()
     Vector Selection Array=DataInterface() {propertyValues {FloatData [<60 130 -70>, <-60 130 -70>]}}
    )
    Sprite Size Mode=::NewEnumerator3
    Uniform Sprite Size=&Multiply_Float(A=77, B=&MakeFloatFromVector(Vector=$Engine.Owner.Scale, Channel=::NewEnumerator0))
    UsePositionOffset=true
    Write Material Random=true
   )
   &InitialMeshOrientation(
    Mesh Orientation Mode=::NewEnumerator7
    Rotation=&MakeVector(X=&RandomRangeFloat(Minimum=0, Maximum=1), Y=0, Z=0)
    Use Rotation Vector=true
   )
   &SkeletalMeshLocation(Bone / Socket Index=3, Bone Sampling Mode=::NewEnumerator4, MaskSpawnGroup=false, Skeletal Mesh=$User.SkeletalMesh)
   &ShapeLocation(
    MaskSpawnGroup=false
    Offset Coordinate Space=::Local
    Offset Mode=::NewEnumerator0
    Override Seed=false
    Ring Radius=50
    Shape Origin=$Particles.SkeletalMeshLocation.SampledPosition
    Shape Primitive=::NewEnumerator4
    UseOwnerScale=true
   )
   &AddVelocity(
    Module.Use Falloff Distance=false
    Override Seed=false
    Velocity Mode=::NewEnumerator1
    Velocity Speed=&RandomRangeFloat(Minimum=200, Maximum=150)
   )
  ]
  particleUpdateScript [
   &ParticleState()
   &ScaleMeshSize(Scale Factor=&VectorFromFloat(Value=&FloatFromCurve(
      FloatCurve=DataInterface() {propertyValues {Curve {keys [
          {tangentMode=RCTM_User, time=0, value=1}
          <at 0.1 2>
          {arriveTangent=-0.0405, interpMode=RCIM_Cubic, leaveTangent=-0.0405, tangentMode=RCTM_User, time=0.3, value=1}
          <at 0.8 0.8>
          {arriveTangent=-1.4118, interpMode=RCIM_Cubic, leaveTangent=-1.4118, tangentMode=RCTM_User, time=1, value=0}
         ]}}}
      Scale Curve=1.5
     ))) {module=ScaleMeshSize002}
   &ScaleMeshSize(Scale Factor=&VectorFromFloat(Value=&FloatFromCurve(
      FloatCurve=DataInterface() {propertyValues {Curve {keys [
          <at 0 0.4 RCIM_Cubic>
          {arriveTangent=-0.0088, interpMode=RCIM_Cubic, leaveTangent=-0.0088, tangentMode=RCTM_User, time=0.7, value=1}
          {arriveTangent=-10.4257, interpMode=RCIM_Cubic, leaveTangent=-10.4257, tangentMode=RCTM_User, time=1, value=0}
         ]}}}
      Scale Curve=&MakeCustomFloatFromBool(Bool=&SetBoolByFloatComparison(A=1, B=0.4), True Float=1.25, False Float=1)
     ))) {enabled=false}
   &ScaleMeshSize(Scale Factor=&MakeVector(
     X=&FloatFromCurve(FloatCurve=DataInterface() {propertyValues {Curve {keys [{interpMode=RCIM_Cubic, tangentMode=RCTM_User, time=0, value=3}, {interpMode=RCIM_Cubic, tangentMode=RCTM_User, time=0.6, value=1}]}}})
     Y=1
     Z=1
    )) {enabled=false, module=ScaleMeshSize001}
   &Color(
    Color=&ColorFromCurve(ColorCurve=DataInterface() {propertyValues {
       AlphaCurve {keys [<at 0.1 1>]}
       BlueCurve {keys [<at 0 0.3643>, {arriveTangent=0.0577, time=0.333, value=0.2745}, <at 1 0.05>]}
       GreenCurve {keys [<at 0 0.3933>, {arriveTangent=-0.0961, time=0.333, value=0.3362}, <at 1 0.05>]}
       RedCurve {keys [<at 0 0.5781>, {arriveTangent=-1.0957, time=0.333, value=0.4583}, <at 1 0.05>]}
      }})
    Scale Alpha=&MakeFloatFromLinearColor(LinearColor=$Particles.Initial.Color, Channel:ENiagaraLinearColor_Channels=::NewEnumerator3)
   )
   &UpdateMeshOrientation(
    Rotation Rate=&FloatFromCurve(
     FloatCurve=DataInterface() {propertyValues {Curve {keys [{interpMode=RCIM_Cubic, tangentMode=RCTM_User, time=0.4, value=0}, {interpMode=RCIM_Cubic, tangentMode=RCTM_User, time=1, value=1}]}}}
     Scale Curve=&RandomRangeFloat(Minimum=-2, Maximum=2)
    )
    Rotation Vector=&RandomVector(Vector Scale=1)
   )
   &DynamicMaterialParameters(
    Index 0 Param 1=&FloatFromCurve(FloatCurve=DataInterface() {propertyValues {Curve {keys [
         {interpMode=RCIM_Cubic, tangentMode=RCTM_User, time=0, value=0}
         {arriveTangent=1.3695, interpMode=RCIM_Cubic, leaveTangent=1.3695, tangentMode=RCTM_User, time=1, value=1}
        ]}}})
    Module.Param0WriteEnabled=true
    Module.Param1WriteEnabled=false
    Module.Param2WriteEnabled=false
    Module.Param3WriteEnabled=false
   ) {enabled=false}
   /SetVariables(Particles.Position=&AddVectorToPosition(
     Position=$Particles.Position
     Vector=&Multiply_VectorByFloat(
      Vector=&NormalizeVector(
       Vector To Normalize=&Lerp_Vector(
        A=&Add_Vector(
         A=<0 1 0>
         B=&Multiply_VectorByFloat(Vector=&RandomVector(Vector Scale=1), Float=0.1)
        )
        B=&TransformVector(
         Vector=&NormalizeVector(
          Vector To Normalize=&Multiply_Vector(A=$Engine.Owner.Velocity, B=<-1 -1 -1>)
          Safe Normalize=true
          Safe Fallback Threshold=0
          Safe Fallback Vector=<0 0 0.001>
         )
         Source Space=::World
         Destination Space=::Local
        )
        Alpha=&FloatFromCurve(FloatCurve=DataInterface() {propertyValues {Curve {keys [
             {interpMode=RCIM_Cubic, leaveTangent=-0.0203, tangentMode=RCTM_Break, time=0, value=0}
             {arriveTangent=-0.0124, interpMode=RCIM_Cubic, leaveTangent=-0.0124, tangentMode=RCTM_User, time=1, value=1}
            ]}}})
       )
       Safe Normalize=true
       Safe Fallback Threshold=0
       Safe Fallback Vector=<0 0 0.001>
      )
      Float=&FloatFromCurve(
       FloatCurve=DataInterface() {propertyValues {Curve {keys [
           {interpMode=RCIM_Cubic, tangentMode=RCTM_User, time=0, value=1.425}
           {arriveTangent=-0.0838, interpMode=RCIM_Cubic, leaveTangent=-0.0838, tangentMode=RCTM_User, time=1, value=0.3}
          ]}}}
       Scale Curve=&Multiply_Float(
        A=5
        B=&MakeCustomFloatFromBool(Bool=&SetBoolByFloatComparison(A=$User.ForwardDot, B=0.4), True Float=1.5, False Float=1)
       )
      )
     )
    )) {enabled=false, isSetParameters=true, module=SetVariables_#1}
   &CalculateAccurateVelocity() {enabled=false}
   &Drag(Drag=&RandomRangeFloat(Minimum=2, Maximum=2), Use Linear Drag=true, Use Rotational Drag=true)
   &SolveForcesAndVelocity(Clamp Velocity=false, Limit Acceleration=false)
  ]
  rendererClasses [&NiagaraMeshRendererProperties]
  renderers [
   {rendererClass=&NiagaraMeshRendererProperties, rendererIndex=0}
   {rendererClass=&NiagaraMeshRendererProperties, rendererIndex=1}
  ]
 }
 Smoke_Toon {
  emitterSpawnScript [
   &NM_ToonShaderParameters(HighLight_Softness=0, Lighting_Hardness=0.8, Metallic_LightingPower=0, RimLight_Width=0)
   &NM_OutlineParameters()
  ]
  emitterUpdateScript [
   &EmitterState()
   &SpawnBurst_Instantaneous(Spawn Count=10, Use Spawn Probability=false, UseLoopCountLimit=false)
   &SpawnRate(Spawn Probability=0.75, SpawnRate=400, Use Spawn Probability=true) {enabled=false}
  ]
  particleSpawnScript [
   &FN_SetSpawnGroup(Number of Spawn Groups=2)
   &InitializeParticle(
    Color=&MakeLinearColorFromVectorAndFloat("Vector (RGB)"=<1 1 1>, "Float (Alpha)"=1)
    Lifetime=&RandomRangeFloat(Minimum=0.6, Maximum=0.8)
    Mesh Scale=&VectorFromFloat(Value=&SelectFloatFromArray(
      Array Sampling Mode=::NewEnumerator0
      Direct Array Index=?
      Float Selection Array=DataInterface() {propertyValues {FloatData [0.3,0.6,0.7]}}
      Random Seed=0
     ))
    Mesh Scale Mode=::NewEnumerator4
    Position Offset=&SelectVectorFromArray(
     Direct Array Index=&SpawnGroup()
     Vector Selection Array=DataInterface() {propertyValues {FloatData [<60 130 -70>, <-60 130 -70>]}}
    )
    Sprite Size Mode=::NewEnumerator3
    Uniform Sprite Size=&Multiply_Float(A=77, B=&MakeFloatFromVector(Vector=$Engine.Owner.Scale, Channel=::NewEnumerator0))
    UsePositionOffset=true
    Write Material Random=true
   )
   &InitialMeshOrientation(
    Mesh Orientation Mode=::NewEnumerator7
    Rotation=&MakeVector(X=&RandomRangeFloat(Minimum=0, Maximum=1), Y=0, Z=0)
    Use Rotation Vector=true
   )
   &SkeletalMeshLocation(Bone Sampling Mode=::NewEnumerator2, MaskSpawnGroup=false, OverrideRandomSeed=false, Skeletal Mesh=$User.SkeletalMesh)
   &ShapeLocation(
    Cylinder Height=1
    Cylinder Radius=1
    MaskSpawnGroup=false
    Override Seed=false
    Shape Origin=$Particles.SkeletalMeshLocation.SampledPosition
    Shape Primitive=::NewEnumerator1
    Surface Only=false
    UseOwnerScale=true
   )
   &AddVelocity(Module.Use Falloff Distance=false, Override Seed=false, Velocity Mode=::NewEnumerator1, Velocity Speed=100)
  ]
  particleUpdateScript [
   &ParticleState()
   &ScaleMeshSize(Scale Factor=&VectorFromFloat(Value=&FloatFromCurve(
      FloatCurve=DataInterface() {propertyValues {Curve {keys [
          {tangentMode=RCTM_User, time=0, value=1}
          <at 0.1 2>
          {arriveTangent=-0.0405, interpMode=RCIM_Cubic, leaveTangent=-0.0405, tangentMode=RCTM_User, time=0.3, value=1}
          <at 0.8 0.8>
          {arriveTangent=-1.4118, interpMode=RCIM_Cubic, leaveTangent=-1.4118, tangentMode=RCTM_User, time=1, value=0}
         ]}}}
      Scale Curve=1.5
     ))) {module=ScaleMeshSize002}
   &ScaleMeshSize(Scale Factor=&VectorFromFloat(Value=&FloatFromCurve(
      FloatCurve=DataInterface() {propertyValues {Curve {keys [
          <at 0 0.4 RCIM_Cubic>
          {arriveTangent=-0.0088, interpMode=RCIM_Cubic, leaveTangent=-0.0088, tangentMode=RCTM_User, time=0.7, value=1}
          {arriveTangent=-10.4257, interpMode=RCIM_Cubic, leaveTangent=-10.4257, tangentMode=RCTM_User, time=1, value=0}
         ]}}}
      Scale Curve=&MakeCustomFloatFromBool(Bool=&SetBoolByFloatComparison(A=1, B=0.4), True Float=1.25, False Float=1)
     ))) {enabled=false}
   &ScaleMeshSize(Scale Factor=&MakeVector(
     X=&FloatFromCurve(FloatCurve=DataInterface() {propertyValues {Curve {keys [{interpMode=RCIM_Cubic, tangentMode=RCTM_User, time=0, value=3}, {interpMode=RCIM_Cubic, tangentMode=RCTM_User, time=0.6, value=1}]}}})
     Y=1
     Z=1
    )) {enabled=false, module=ScaleMeshSize001}
   &Color(
    Color=&ColorFromCurve(ColorCurve=DataInterface() {propertyValues {
       AlphaCurve {keys [<at 0.1 1>]}
       BlueCurve {keys [<at 0 0.3643>, {arriveTangent=0.0577, time=0.333, value=0.2745}, <at 1 0.05>]}
       GreenCurve {keys [<at 0 0.3933>, {arriveTangent=-0.0961, time=0.333, value=0.3362}, <at 1 0.05>]}
       RedCurve {keys [<at 0 0.5781>, {arriveTangent=-1.0957, time=0.333, value=0.4583}, <at 1 0.05>]}
      }})
    Scale Alpha=&MakeFloatFromLinearColor(LinearColor=$Particles.Initial.Color, Channel:ENiagaraLinearColor_Channels=::NewEnumerator3)
   )
   &UpdateMeshOrientation(
    Rotation Rate=&FloatFromCurve(
     FloatCurve=DataInterface() {propertyValues {Curve {keys [{interpMode=RCIM_Cubic, tangentMode=RCTM_User, time=0.4, value=0}, {interpMode=RCIM_Cubic, tangentMode=RCTM_User, time=1, value=1}]}}}
     Scale Curve=&RandomRangeFloat(Minimum=-2, Maximum=2)
    )
    Rotation Vector=&RandomVector(Vector Scale=1)
   )
   &DynamicMaterialParameters(
    Index 0 Param 1=&FloatFromCurve(FloatCurve=DataInterface() {propertyValues {Curve {keys [
         {interpMode=RCIM_Cubic, tangentMode=RCTM_User, time=0, value=0}
         {arriveTangent=1.3695, interpMode=RCIM_Cubic, leaveTangent=1.3695, tangentMode=RCTM_User, time=1, value=1}
        ]}}})
    Module.Param0WriteEnabled=true
    Module.Param1WriteEnabled=false
    Module.Param2WriteEnabled=false
    Module.Param3WriteEnabled=false
   ) {enabled=false}
   /SetVariables(Particles.Position=&AddVectorToPosition(
     Position=$Particles.Position
     Vector=&Multiply_VectorByFloat(
      Vector=&NormalizeVector(
       Vector To Normalize=&Lerp_Vector(
        A=&Add_Vector(
         A=<0 1 0>
         B=&Multiply_VectorByFloat(Vector=&RandomVector(Vector Scale=1), Float=0.1)
        )
        B=&TransformVector(
         Vector=&NormalizeVector(
          Vector To Normalize=&Multiply_Vector(A=$Engine.Owner.Velocity, B=<-1 -1 -1>)
          Safe Normalize=true
          Safe Fallback Threshold=0
          Safe Fallback Vector=<0 0 0.001>
         )
         Source Space=::World
         Destination Space=::Local
        )
        Alpha=&FloatFromCurve(FloatCurve=DataInterface() {propertyValues {Curve {keys [
             {interpMode=RCIM_Cubic, leaveTangent=-0.0203, tangentMode=RCTM_Break, time=0, value=0}
             {arriveTangent=-0.0124, interpMode=RCIM_Cubic, leaveTangent=-0.0124, tangentMode=RCTM_User, time=1, value=1}
            ]}}})
       )
       Safe Normalize=true
       Safe Fallback Threshold=0
       Safe Fallback Vector=<0 0 0.001>
      )
      Float=&FloatFromCurve(
       FloatCurve=DataInterface() {propertyValues {Curve {keys [
           {interpMode=RCIM_Cubic, tangentMode=RCTM_User, time=0, value=1.425}
           {arriveTangent=-0.0838, interpMode=RCIM_Cubic, leaveTangent=-0.0838, tangentMode=RCTM_User, time=1, value=0.3}
          ]}}}
       Scale Curve=&Multiply_Float(
        A=5
        B=&MakeCustomFloatFromBool(Bool=&SetBoolByFloatComparison(A=$User.ForwardDot, B=0.4), True Float=1.5, False Float=1)
       )
      )
     )
    )) {enabled=false, isSetParameters=true, module=SetVariables_#1}
   &CalculateAccurateVelocity() {enabled=false}
   &SolveForcesAndVelocity(Clamp Velocity=false, Limit Acceleration=false)
  ]
  rendererClasses [&NiagaraMeshRendererProperties]
  renderers [
   {rendererClass=&NiagaraMeshRendererProperties, rendererIndex=0}
   {rendererClass=&NiagaraMeshRendererProperties, rendererIndex=1}
  ]
 }
 airflow001 {
  emitterUpdateScript [
   &EmitterState(Life Cycle Mode=::NewEnumerator1, Loop Behavior=::NewEnumerator1, Loop Delay=1, Loop Duration=1, UseLoopDelay=false)
   &SpawnBurst_Instantaneous(Spawn Count=6, Use Spawn Probability=false, UseLoopCountLimit=false)
  ]
  particleSpawnScript [
   &InitializeParticle(
    Color=<rgba 1 1 1 0.2>
    Lifetime Max=0.6
    Lifetime Min=0.4
    Lifetime Mode=::NewEnumerator1
    Mesh Scale=&Multiply_VectorByFloat(
     Vector=&RandomRangeVector(Minimum=<1 1 1>, Maximum=<1 1 1.5>)
     Float=&RandomRangeFloat(Minimum=0.5, Maximum=0.9)
    )
    Mesh Scale Mode=::NewEnumerator4
    UsePositionOffset=false
    Write Material Random=true
   )
   &InitialMeshOrientation(
    Mesh Orientation Mode=::NewEnumerator2
    Orientation Coordinate Space=::Local
    Rotation=&MakeVector(
     X=&MakeCustomFloatFromBool(Bool=&RandomBool(), True Float=-0.03, False Float=0.03)
     Y=&MakeCustomFloatFromBool(Bool=&RandomBool(), True Float=-0.03, False Float=0.03)
     Z=&RandomRangeFloat(Minimum=-1, Maximum=1)
    )
    Use Rotation Vector=true
   )
   &SystemLocation(MaskSpawnGroup=false, Offset=&RandomRangeVector(Minimum=<0 0 0>, Maximum=<0 0 150>))
  ]
  particleUpdateScript [
   &ParticleState()
   &DynamicMaterialParameters(
    Index 0 Param 1=&FloatFromCurve(FloatCurve=DataInterface() {propertyValues {Curve {keys [{interpMode=RCIM_Cubic, tangentMode=RCTM_User, time=0, value=1}, {interpMode=RCIM_Cubic, tangentMode=RCTM_User, time=1, value=0}]}}})
    Module.Param0WriteEnabled=true
    Module.Param1WriteEnabled=true
    Module.Param2WriteEnabled=true
    Module.Param3WriteEnabled=true
   )
   &ScaleMeshSize(Scale Factor=&Multiply_VectorByFloat(
     Vector=<1 1 1>
     Float=&FloatFromCurve(
      FloatCurve=DataInterface() {propertyValues {Curve {keys [
          {arriveTangent=2.3149, interpMode=RCIM_Cubic, leaveTangent=2.3149, tangentMode=RCTM_User, time=0, value=0}
          {arriveTangent=0.9995, interpMode=RCIM_Cubic, leaveTangent=0.9995, tangentMode=RCTM_User, time=0.3, value=0.45}
          {arriveTangent=-0.0405, interpMode=RCIM_Cubic, leaveTangent=-0.0405, tangentMode=RCTM_User, time=1, value=1}
         ]}}}
      Scale Curve=2.5
     )
    ))
   &UpdateMeshOrientation(
    Rotation Rate=&MakeCustomFloatFromBool(Bool=&RandomBool(), True Float=2.5, False Float=-2.5)
    Rotation Vector=<0 0 1>
   )
   &ScaleColor(
    Scale Alpha=&FloatFromCurve(FloatCurve=DataInterface() {propertyValues {Curve {keys [
         {interpMode=RCIM_Cubic, tangentMode=RCTM_User, time=0, value=0}
         {interpMode=RCIM_Cubic, leaveTangent=-8.2902, tangentMode=RCTM_Break, time=0.3, value=1}
         {tangentMode=RCTM_User, time=1, value=0}
        ]}}})
    Scale RGB=&Multiply_VectorByFloat(Vector=<1 1 1>, Float=&FN_FrontEndFloatModifier(Gameplay=1, Front End=2))
    ScaleA=true
    ScaleRGB=true
   )
   /SetVariables(Particles.Mesh soft color:LinearColor=&MakeLinearColorFromVectorAndFloat(
     "Vector (RGB)"=&Multiply_VectorByFloat(Vector=<1 1 1>, Float=&FN_FrontEndFloatModifier(Gameplay=1, Front End=2))
     "Float (Alpha)"=&Multiply_Float(
      A=&FloatFromCurve(FloatCurve=DataInterface() {propertyValues {Curve {keys [
           {interpMode=RCIM_Cubic, tangentMode=RCTM_User, time=0, value=0}
           {interpMode=RCIM_Cubic, tangentMode=RCTM_User, time=0.5, value=1}
           {tangentMode=RCTM_User, time=1, value=0}
          ]}}})
      B=1
     )
    )) {isSetParameters=true}
  ]
  rendererClasses [&NiagaraMeshRendererProperties]
  renderers [
   {rendererClass=&NiagaraMeshRendererProperties, rendererIndex=0}
   {rendererClass=&NiagaraMeshRendererProperties, rendererIndex=1}
  ]
 }
 spikes {
  emitterUpdateScript [
   &EmitterState(Life Cycle Mode=::NewEnumerator1, Loop Behavior=::NewEnumerator1, Loop Delay=0.03, Loop Duration=1, UseLoopDelay=true)
   &SpawnBurst_Instantaneous(Spawn Count=10, Use Spawn Probability=false, UseLoopCountLimit=false)
  ]
  particleSpawnScript [
   &InitializeParticle(
    Color=<rgba 1 0.6933 0.2 1>
    Lifetime Max=0.25
    Lifetime Min=0.15
    Lifetime Mode=::NewEnumerator1
    Sprite Size Max=<100 180>
    Sprite Size Min=<20 100>
    Sprite Size Mode=::NewEnumerator2
    UsePositionOffset=false
    Write Material Random=true
   )
   &TorusLocation(Align Torus With Axis=false, Handle Radius=0, Large Radius=10, MaskSpawnGroup=false, Module.Override Randomness=false)
   &AddVelocityFromPoint(
    Module.Use Falloff Distance=false
    Origin Offset=&RandomRangeVector(Minimum=<0 0 -5>, Maximum=<0 0 -10>)
    Velocity Strength=&RandomRangeFloat(Minimum=600, Maximum=800)
   )
   &SubUVAnimation(End Frame=0, SubUV Animation Mode:ENiagara_SubUVLookupMode=::NewEnumerator2)
  ]
  particleUpdateScript [
   &ParticleState()
   &Drag(Drag=6, Use Linear Drag=true, Use Rotational Drag=false)
   &SolveForcesAndVelocity(Clamp Velocity=false, Limit Acceleration=false)
   &ScaleSpriteSize(
    Non-Uniform Curve Scale=&Multiply_Vector2D_ByFloat(Float=0.4, Vector2D=<1 1>)
    Non-Uniform Curve Sprite Scale=DataInterface() {propertyValues {
      XCurve {keys [{tangentMode=RCTM_User, time=0, value=1}, {tangentMode=RCTM_User, time=1, value=0.8}]}
      YCurve {keys [{tangentMode=RCTM_User, time=0, value=0.5}, {tangentMode=RCTM_User, time=1, value=1.7}]}
     }}
    Scale Sprite Size Mode=::NewEnumerator3
   )
   &ScaleColor(
    Scale Alpha=&FloatFromCurve(
     FloatCurve=DataInterface() {propertyValues {Curve {keys [<at 0 0>,<at 0.1 1>,<at 0.2 0.8>,<at 1 0>]}}}
     Scale Curve=&FNTimeOfDayScalar(Night Scale=1, Morning Scale=0.3, Evening Scale=0.5)
    )
    ScaleA=true
    ScaleRGB=true
   )
  ]
  rendererClasses [&NiagaraSpriteRendererProperties]
  renderers [{rendererClass=&NiagaraSpriteRendererProperties}]
 }
}
summary {
 systemName=NS_Emote_EmeraldGlassTransform_SkinnyMight
 emitters [
  {emitterName=Smoke_Toon, rendererClasses [&NiagaraMeshRendererProperties]}
  {emitterName=BigSmok, rendererClasses [&NiagaraMeshRendererProperties]}
  {emitterName=SmokeRing, rendererClasses [&NiagaraMeshRendererProperties]}
  {emitterName=airflow001, rendererClasses [&NiagaraMeshRendererProperties]}
  {emitterName=spikes, rendererClasses [&NiagaraSpriteRendererProperties]}
  {emitterName=HQ_Light, rendererClasses [&NiagaraLightRendererProperties]}
 ]
 userVariables [
  User.bIsFrontEnd {type=NiagaraBool}
  User.bIsFrontEndPreview {type=NiagaraBool}
  User.ForwardDot {type=NiagaraFloat, defaultValue=0, description=""}
  User.FrontendSpawn {type=NiagaraBool}
  User.RibbonIDOffset {type=NiagaraInt32}
  User.SkeletalMesh {
   type=NiagaraDataInterfaceSkeletalMesh
   defaultValue {
    struct=NiagaraExt_VariableValue_DataInterface
    value {dataInterface=&self:NiagaraDataInterfaceSkeletalMesh_1}
   }
  }
 ]
}
